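# Idle Niagara effect for The Origin backpack cosmetic, layering looping cube particles, glow sprites and a flickering light.
# Light and decal emitters re-sample bones or sockets on User.SkeletalMesh every update and pulse through waveform-driven alpha.
# Cube emitters mix mesh and sprite renderers: one shoots upward against drag, another spawns on two offset sides and swirls by vortex and curve-scaled velocity.
# Curves over particle age drive size, color and dynamic material parameters; Cubes_CrashCenter is disabled.
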NiagaraSystem /BRCosmetics/Accessories/FORT_Backpacks/Backpack_M_MED_Origin_Prison/FX/NS_Pack_The_Origin_Idle
dependencies {
 usedDataInterfaces [
  &NiagaraDataInterfaceArrayFloat
  &NiagaraDataInterfaceCurve
  &NiagaraDataInterfaceVectorCurve
  &NiagaraDataInterfaceColorCurve
  &NiagaraDataInterfaceArrayInt32
  &NiagaraDataInterfaceArrayFloat3
  &NiagaraDataInterfaceMeshRendererInfo
  &NiagaraDataInterfaceParticleRead
 ]
 usedDynamicInputs []
 usedModules [&SystemState]
}
emitters {
 CenterLightFlicker {
  emitterUpdateScript [
   &EmitterState(Life Cycle Mode=::NewEnumerator1, Loop Behavior=::NewEnumerator1, Loop Delay=0.8, Loop Duration Mode=::NewEnumerator1, UseLoopDelay=true)
   &SpawnBurst_Instantaneous(Use Spawn Probability=false, UseLoopCountLimit=false)
   &SpawnBurst_Instantaneous(Spawn Group=1, Use Spawn Probability=false, UseLoopCountLimit=false) {enabled=false, module=SpawnBurst_Instantaneous001}
  ]
  particleSpawnScript [
   &InitializeParticle(
    Color=<rgba 0.2053 0.0464 1 1>
    Sprite Size Mode=::NewEnumerator3
    Uniform Sprite Size=12
    UsePositionOffset=true
    Write Material Random=true
   )
   &SkeletalMeshLocation(
    Bone / Socket Index=2
    Bone Sampling Mode=::NewEnumerator4
    MaskSpawnGroup=false
    Orientation Sampling=::NewEnumerator1
    Position Sampling=::NewEnumerator2
    Sampled Position Offset=<-0.08 0 0>
    Skeletal Mesh=$User.SkeletalMesh
   ) {module=SkeletalMeshLocation001}
  ]
  particleUpdateScript [
   &SystemLocation(MaskSpawnGroup=false) {enabled=false}
   &ParticleState(
    Kill Particles When Lifetime Has Elapsed=false
    Let Infinitely Lived Particles Die When Emitter Deactivates=true
    Loop Particles Lifetime=true
   )
   &SkeletalMeshLocation(
    Bone / Socket Index=2
    Bone Sampling Mode=::NewEnumerator4
    MaskSpawnGroup=false
    Orientation Sampling=::NewEnumerator1
    Position Sampling=::NewEnumerator2
    Sampled Position Offset=<-0.08 5 0>
    Skeletal Mesh=$User.SkeletalMesh
   )
   &ScaleColor(
    Scale Alpha=&Waveform(
     Waveform Count=::NewEnumerator1
     "Waveform [1]"=::NewEnumerator0
     "[1] Frequency"=5
     "[1] Amplitude Min/Max"=<0 0.5>
     "[1] Bias"=0.2
     "[1] Pulse Duration"=0.3
     "[2] Average"=false
     "[2] Input Position"=0
     "[2] Phase"=$Emitter.Age
     "[2] Phase Scale"=0.3
     "[2] Compound Frequency"=2
     "[2] Frequency"=5
     "[2] Amplitude Min/Max"=<0 1>
     "[2] Scale"=1
     "[2] Bias"=0.5
     "[2] Exponent"=1
     "[2] Pulse Duration"=0.5
     "[2] Clip Triangle Min/Max"=<0 0>
     "[2] Random Seed"=0
    )
    ScaleA=true
    ScaleRGB=true
   )
   /SetVariables(
    Particles.ColorLight:LinearColor=&MultiplyLinearColor(
     Linear Color=$Particles.Color
     Scale Factor=&MakeFloat4FromVectorAndFloat(
      "Vector (RGB)"=&VectorFromFloat(Value=&RandomRangeFloat(Minimum=3, Maximum=3))
      "Float (Alpha)"=2
     )
    )
    Particles.ColorSprites:LinearColor=&MultiplyLinearColor(
     Linear Color=$Particles.Color
     Scale Factor=&MakeFloat4FromVectorAndFloat("Vector (RGB)"=&VectorFromFloat(Value=3), "Float (Alpha)"=1)
    )
    Particles.LightRadius=&RandomRangeFloat(Minimum=80, Maximum=80, Evaluation Type=::NewEnumerator1)
    Particles.Position_Part1:NiagaraPosition=&ConvertVectorToPosition(
     Input Position=Linked() {linkedVariable {name=Particles.SkeletalMeshLocation.SampledPosition, type=NiagaraPosition}}
     Vector Is In Absolute Large World Space:NiagaraBool=true
    )
    Particles.Position_Part2:NiagaraPosition=&ConvertVectorToPosition(
     Input Position=Linked() {linkedVariable {name=Particles.SkeletalMeshLocation001.SampledPosition, type=NiagaraPosition}}
     Vector Is In Absolute Large World Space:NiagaraBool=true
    )
   ) {isSetParameters=true, module=SetVariables_#1}
  ]
  rendererClasses [&NiagaraLightRendererProperties]
  renderers [{rendererClass=&NiagaraLightRendererProperties, rendererIndex=0}]
 }
 CubeCrashDecals {
  emitterUpdateScript [
   &EmitterState(Life Cycle Mode=::NewEnumerator1, Loop Behavior=::NewEnumerator1, Loop Delay=0.3, Loop Duration Mode=::NewEnumerator1, UseLoopDelay=true)
   &SpawnBurst_Instantaneous(Spawn Count=2, Use Spawn Probability=false, UseLoopCountLimit=false)
   &SpawnBurst_Instantaneous(Spawn Group=1, Use Spawn Probability=false, UseLoopCountLimit=false) {enabled=false, module=SpawnBurst_Instantaneous001}
  ]
  particleSpawnScript [
   &PartitionParticles()
   &InitializeParticle(Color=<rgba 1 0 0.4744 1>, Sprite Size Mode=::NewEnumerator1, Uniform Sprite Size Min=8, UsePositionOffset=true, Write Material Random=true)
  ]
  particleUpdateScript [
   &SystemLocation(MaskSpawnGroup=false) {enabled=false}
   &ParticleState(
    Kill Particles When Lifetime Has Elapsed=false
    Let Infinitely Lived Particles Die When Emitter Deactivates=true
    Loop Particles Lifetime=true
   )
   &SpriteFacingAndAlignment(
    Facing Coordinate Space=::Local
    Sprite Facing=&SelectVectorFromArray(
     Direct Array Index=$Particles.PartitionParticles.Partition
     Vector Selection Array=DataInterface() {propertyValues {FloatData [<0 -1 0>, <0 1 0>, <0 0 0>]}}
    )
    WriteAlignment=false
    WriteFacing=true
   )
   &SkeletalMeshLocation(
    Bone / Socket Index=$Particles.PartitionParticles.Partition
    Bone Sampling Mode=::NewEnumerator4
    MaskSpawnGroup=false
    Orientation Sampling=::NewEnumerator1
    Position Sampling=::NewEnumerator2
    Sampled Position Offset=<-9 0 0>
    Skeletal Mesh=$User.SkeletalMesh
   )
   &SkeletalMeshLocation(
    Bone / Socket Index=$Particles.PartitionParticles.Partition
    Bone Sampling Mode=::NewEnumerator4
    MaskSpawnGroup=false
    Orientation Sampling=::NewEnumerator1
    Position Sampling=::NewEnumerator2
    Sampled Position Offset=<-2 0 0>
    Skeletal Mesh=$User.SkeletalMesh
   ) {module=SkeletalMeshLocation001}
   &ScaleColor(
    Scale Alpha=&Waveform(
     Waveform Count=::NewEnumerator1
     "Waveform [1]"=::NewEnumerator5
     "[1] Frequency"=5
     "[1] Amplitude Min/Max"=<0 0.5>
     "[1] Bias"=0.2
     "[1] Pulse Duration"=0.5
     "[2] Average"=false
     "Waveform [2]"=::NewEnumerator5
     "[2] Input Position"=0
     "[2] Phase"=$Emitter.Age
     "[2] Phase Scale"=0.3
     "[2] Compound Frequency"=2
     "[2] Frequency"=5
     "[2] Amplitude Min/Max"=<0 1>
     "[2] Scale"=1
     "[2] Bias"=0
     "[2] Exponent"=1
     "[2] Pulse Duration"=0.1
     "[2] Clip Triangle Min/Max"=<0 0>
     "[2] Random Seed"=0
    )
    ScaleA=true
    ScaleRGB=true
   )
   /SetVariables(
    Particles.ColorLight:LinearColor=&MultiplyLinearColor(
     Linear Color=$Particles.Color
     Scale Factor=&MakeFloat4FromVectorAndFloat(
      "Vector (RGB)"=&VectorFromFloat(Value=&RandomRangeFloat(Minimum=1, Maximum=1))
      "Float (Alpha)"=1
     )
    )
    Particles.ColorSprites:LinearColor=&MultiplyLinearColor(
     Linear Color=$Particles.Color
     Scale Factor=&MakeFloat4FromVectorAndFloat("Vector (RGB)"=&VectorFromFloat(Value=3), "Float (Alpha)"=1)
    )
    Particles.LightRadius=&RandomRangeFloat(Minimum=80, Maximum=80, Evaluation Type=::NewEnumerator1)
    Particles.Position_Part1:NiagaraPosition=&ConvertVectorToPosition(
     Input Position=Linked() {linkedVariable {name=Particles.SkeletalMeshLocation.SampledPosition, type=NiagaraPosition}}
     Vector Is In Absolute Large World Space:NiagaraBool=true
    )
    Particles.Position_Part2:NiagaraPosition=&ConvertVectorToPosition(
     Input Position=Linked() {linkedVariable {name=Particles.SkeletalMeshLocation001.SampledPosition, type=NiagaraPosition}}
     Vector Is In Absolute Large World Space:NiagaraBool=true
    )
   ) {isSetParameters=true, module=SetVariables_#1}
   &DynamicMaterialParameters(
    Index 0 Param 1=2
    Module.Param0WriteEnabled=true
    Module.Param1WriteEnabled=true
    Module.Param2WriteEnabled=true
    Module.Param3WriteEnabled=true
   )
   &CameraOffset(Camera Offset Amount=8)
  ]
  rendererClasses [&NiagaraSpriteRendererProperties]
  renderers [
   {rendererClass=&NiagaraSpriteRendererProperties}
   {rendererClass=&NiagaraSpriteRendererProperties, rendererIndex=1}
  ]
 }
 CubesShootUpBurst {
  emitterUpdateScript [
   &EmitterState(Life Cycle Mode=::NewEnumerator1, Loop Behavior=::NewEnumerator1, Loop Duration=0.8, UseLoopDelay=false)
   &SpawnRate(
    Spawn Probability=&FloatFromCurve(
     FloatCurve=DataInterface() {propertyValues {Curve {keys [
         {interpMode=RCIM_Cubic, tangentMode=RCTM_User, time=0.2, value=0.5}
         {interpMode=RCIM_Cubic, tangentMode=RCTM_User, time=0.5, value=1}
         {tangentMode=RCTM_User, time=0.8, value=0.1}
        ]}}}
     CurveIndex=$Emitter.NormalizedLoopAge
    )
    SpawnRate=60
    Use Spawn Probability=true
   )
  ]
  particleSpawnScript [
   &InitializeParticle(
    Color=&MakeLinearColorFromVectorAndFloat(
     "Vector (RGB)"=&VectorFromFloat(Value=&RandomRangeFloat(Minimum=0.1, Maximum=0.3))
     "Float (Alpha)"=1
    )
    Lifetime=&RandomRangeFloat(Minimum=0.4, Maximum=0.5)
    Mesh Scale Max=<2 0.5 0.7>
    Mesh Scale Min=<2 0.5 0.3>
    Mesh Scale Mode=::NewEnumerator2
    Sprite Size Mode=::NewEnumerator3
    Uniform Sprite Size=&RandomRangeFloat(Minimum=5, Maximum=7)
    UsePositionOffset=false
    Write Material Random=true
   )
   &ShapeLocation(
    MaskSpawnGroup=false
    Offset Coordinate Space=::Local
    Offset Mode=::NewEnumerator0
    Override Seed=false
    Rotation Mode=::NewEnumerator2
    Shape Origin=$Engine.Emitter.SimulationPosition
    Sphere Radius=5
    UseOwnerScale=true
    Yaw / Pitch / Roll=<0 90 90>
   ) {module=ShapeLocation001}
   &AddVelocity(
    Module.Use Falloff Distance=false
    Override Seed=false
    Velocity Mode=::NewEnumerator1
    Velocity Speed=&RandomRangeFloat(Minimum=5, Maximum=20)
   ) {enabled=false, module=AddVelocity001}
   &InheritVelocity(Inherited Velocity Amount Scale=<0.5 0.5 0.5>, Inherited Velocity Speed Limit=3, Speed Limit=true, UseSpeedThreshold=false) {enabled=false}
   &ApplyInitialForces(Apply Force to Position=true)
   &AddVelocity(Rotation Coordinate Space=::World, Velocity=<0 0 320>)
  ]
  particleUpdateScript [
   &ParticleState()
   &ShapeLocation(
    Cylinder Height=20
    Cylinder Radius=5
    MaskSpawnGroup=false
    Override Seed=false
    Rotation Mode=::NewEnumerator2
    Shape Primitive=::NewEnumerator1
    Surface Only=true
    Surface Only Band Thickness:NiagaraFloat=&FloatFromCurve(FloatCurve=DataInterface() {propertyValues {Curve {keys [{tangentMode=RCTM_User, time=0, value=0}, {tangentMode=RCTM_User, time=1, value=1}]}}})
    Use Endcaps In Surface Only Mode=false
    UseOwnerScale=true
    Yaw / Pitch / Roll=<0 90 90>
   ) {enabled=false}
   /Niagara/Modules/Update/Velocity/VortexVelocity(
    Module.Use Influence Falloff=false
    Velocity Amount:NiagaraFloat=&FloatFromCurve(
     FloatCurve=DataInterface() {propertyValues {Curve {keys [{interpMode=RCIM_Cubic, tangentMode=RCTM_User, time=0, value=1}, {interpMode=RCIM_Cubic, tangentMode=RCTM_User, time=1, value=0}]}}}
     Scale Curve=2000
    )
    Vortex Axis:Vector3f=<0 1 0>
   ) {enabled=false}
   &ScaleVelocity(Velocity Scale=&MakeVector(
     X=&FloatFromCurve(
      FloatCurve=DataInterface() {propertyValues {Curve {keys [
          {arriveTangent=4.4437, interpMode=RCIM_Cubic, leaveTangent=4.4437, tangentMode=RCTM_User, time=0.5, value=0}
          {arriveTangent=-0.8586, interpMode=RCIM_Cubic, leaveTangent=-0.8586, tangentMode=RCTM_User, time=1, value=1}
         ]}}}
      Scale Curve=-6
     )
     Y=&FloatFromCurve(
      FloatCurve=DataInterface() {propertyValues {Curve {keys [
          {interpMode=RCIM_Cubic, tangentMode=RCTM_User, time=0, value=0.4}
          {interpMode=RCIM_Cubic, tangentMode=RCTM_User, time=0.1517, value=1.1096}
          <at 0.3922 1.2096>
          {tangentMode=RCTM_User, time=0.5922, value=0.1}
         ]}}}
      Scale Curve=3
     )
     Z=&FloatFromCurve(
      FloatCurve=DataInterface() {propertyValues {Curve {keys [
          {arriveTangent=6.7878, interpMode=RCIM_Cubic, leaveTangent=6.7878, tangentMode=RCTM_User, time=0.5, value=0}
          {arriveTangent=-1.0738, interpMode=RCIM_Cubic, leaveTangent=-1.0738, tangentMode=RCTM_User, time=1, value=1}
         ]}}}
      Scale Curve=-4
     )
    )) {enabled=false}
   &Drag(Drag=20, Rotational Drag=8, Use Linear Drag=true, Use Rotational Drag=false)
   &AccelerationForce(Acceleration=<0 0 100>) {enabled=false}
   &SolveForcesAndVelocity(Clamp Velocity=false, Limit Acceleration=false)
   &ScaleMeshSize(Scale Factor=&MakeVector(
     X=&FloatFromCurve(
      FloatCurve=DataInterface() {propertyValues {Curve {keys [
          <at 0 0.5>
          {arriveTangent=1.0382, interpMode=RCIM_Cubic, leaveTangent=1.0382, tangentMode=RCTM_User, time=0.2, value=0.2}
          {interpMode=RCIM_Cubic, tangentMode=RCTM_User, time=1, value=0.4}
         ]}}}
      Scale Curve=3
     )
     Y=&FloatFromCurve(FloatCurve=DataInterface() {propertyValues {Curve {keys [
          {arriveTangent=-5.7231, interpMode=RCIM_Cubic, leaveTangent=-5.7231, tangentMode=RCTM_User, time=0, value=1}
          {arriveTangent=11.6115, interpMode=RCIM_Cubic, leaveTangent=11.6115, time=0.1411, value=5.2397}
          <at 0.2821 4.2757>
          {arriveTangent=2.5208, interpMode=RCIM_Cubic, leaveTangent=2.5208, tangentMode=RCTM_User, time=0.39, value=2.3196}
         ]}}})
     Z=&FloatFromCurve(
      FloatCurve=DataInterface() {propertyValues {Curve {keys [
          <at 0 0.5>
          {arriveTangent=0.9168, interpMode=RCIM_Cubic, leaveTangent=0.9168, tangentMode=RCTM_User, time=0.2, value=0.2}
          {interpMode=RCIM_Cubic, tangentMode=RCTM_User, time=1, value=0.8}
         ]}}}
      Scale Curve=3
     )
    )) {enabled=false}
   &ScaleMeshSize(Scale Factor=&Multiply_Vector(
     A=&MakeVector(
      X=&FloatFromCurve(
       FloatCurve=DataInterface() {propertyValues {Curve {keys [
           {interpMode=RCIM_Cubic, tangentMode=RCTM_User, time=0, value=0}
           {interpMode=RCIM_Cubic, tangentMode=RCTM_User, time=0.5, value=1}
           {tangentMode=RCTM_User, time=1, value=0}
          ]}}}
       Scale Curve=2
      )
      Y=1
      Z=1
     )
     B=&VectorFromFloat(Value=&FloatFromCurve(
       FloatCurve=DataInterface() {propertyValues {Curve {keys [
           {interpMode=RCIM_Cubic, tangentMode=RCTM_User, time=0.2, value=0.1}
           {arriveTangent=2.5962, interpMode=RCIM_Cubic, leaveTangent=2.5962, tangentMode=RCTM_User, time=1, value=1}
          ]}}}
       Scale Curve=6
      ))
    )) {enabled=false, module=ScaleMeshSize002}
   &Color(Color=&ColorFromCurve(
     ColorCurve=DataInterface() {propertyValues {
       AlphaCurve {keys [<at 0 1 RCIM_Cubic>,<at 1 1 RCIM_Cubic>]}
       BlueCurve {keys [<at 0 0>,<at 0.1115 0>,<at 0.2661 0 RCIM_Cubic>,<at 0.3914 0>,<at 0.683 0>,<at 0.865 0 RCIM_Cubic>]}
       GreenCurve {keys [
         <at 0 0.1667>
         <at 0.1115 0.6667>
         {arriveTangent=0.5956, interpMode=RCIM_Cubic, leaveTangent=0.5956, time=0.2661, value=0.0833}
         <at 0.3914 0.8333>
         <at 0.683 0.8333>
         <at 0.865 0.1667 RCIM_Cubic>
        ]}
       RedCurve {keys [
         <at 0 5>
         <at 0.1115 20>
         {arriveTangent=-35.7343, interpMode=RCIM_Cubic, leaveTangent=-35.7343, time=0.2661, value=1}
         <at 0.3914 10>
         <at 0.683 10>
         <at 0.865 5 RCIM_Cubic>
        ]}
      }}
     CurveIndex=&Multiply_Float(A=&RandomRangeFloat(Minimum=0, Maximum=1), B=$Particles.NormalizedAge)
    )) {enabled=false}
   &ScaleSpriteSize(
    Uniform Curve Scale=4
    Uniform Curve Sprite Scale=DataInterface() {propertyValues {Curve {keys [
        {interpMode=RCIM_Cubic, tangentMode=RCTM_User, time=0, value=0}
        {arriveTangent=6.9659, interpMode=RCIM_Cubic, leaveTangent=6.9659, time=0.0359, value=0.759}
        {arriveTangent=-0.3574, interpMode=RCIM_Cubic, leaveTangent=-0.3574, time=0.0877, value=0.611}
        <at 0.3079 0.6618>
        {arriveTangent=-0.4523, interpMode=RCIM_Cubic, leaveTangent=-0.4523, time=0.5028, value=0.3018}
        {arriveTangent=0.3301, interpMode=RCIM_Cubic, leaveTangent=0.3301, time=0.8828, value=0.4018}
        {arriveTangent=-1.7215, interpMode=RCIM_Cubic, leaveTangent=-1.7215, time=0.9507, value=0.4496}
        {arriveTangent=2.875, interpMode=RCIM_Cubic, leaveTangent=2.875, tangentMode=RCTM_User, time=1, value=0.2}
       ]}}}
   )
   &ScaleColor(
    Scale RGB=&MakeVectorFromLinearColorRGB(Color=&ColorFromCurve(
      ColorCurve=DataInterface() {propertyValues {
        AlphaCurve {keys [<at 0 1 RCIM_Cubic>,<at 1 1 RCIM_Cubic>]}
        BlueCurve {keys [{arriveTangent=-0.2836, interpMode=RCIM_Cubic, time=0, value=0.2}, <at 0.1682 0.5>, <at 0.2822 0.5>, <at 0.3177 0>, <at 0.7877 0>, <at 1 0>]}
        GreenCurve {keys [
          {arriveTangent=-0.3781, interpMode=RCIM_Cubic, time=0, value=0.0216}
          <at 0.1682 0>
          <at 0.2822 0>
          <at 0.3177 0.1667>
          <at 0.7877 0.25>
          <at 1 0.0333>
         ]}
        RedCurve {keys [
          <at 0 0.0811 RCIM_Cubic>
          <at 0.1682 0.1667>
          <at 0.2822 0.4167>
          {arriveTangent=21.5915, leaveTangent=21.5915, time=0.3177, value=1}
          <at 0.7877 1>
          <at 1 1>
         ]}
       }}
      CurveIndex=&Multiply_Float(A=&RandomRangeFloat(Minimum=1, Maximum=1), B=$Particles.NormalizedAge)
     ))
    ScaleA=true
    ScaleRGB=true
   )
   &DynamicMaterialParameters(
    Index 0 Param 1=&FloatFromCurve(FloatCurve=DataInterface() {propertyValues {Curve {keys [<at 0 1>,<at 0.0833 3.8621>,<at 0.1534 3.5031>,<at 0.2 2.5>,<at 0.7215 2>,<at 0.981 15.7573>,<at 1.0836 27.0214>,<at 1.1836 27.5214>]}}})
    Index 0 Param 3=&FloatFromCurve(
     FloatCurve=DataInterface() {propertyValues {Curve {keys [
         {arriveTangent=-0.5377, interpMode=RCIM_Cubic, leaveTangent=-0.5377, tangentMode=RCTM_User, time=0, value=1}
         {arriveTangent=-2.3485, interpMode=RCIM_Cubic, leaveTangent=-2.3485, tangentMode=RCTM_User, time=0.7424, value=0.55}
         {arriveTangent=-3.0281, interpMode=RCIM_Cubic, leaveTangent=-3.0281, tangentMode=RCTM_User, time=1, value=0}
        ]}}}
     Scale Curve=-1
    )
    Module.Param0WriteEnabled=true
    Module.Param1WriteEnabled=true
    Module.Param2WriteEnabled=true
    Module.Param3WriteEnabled=true
   )
   /SetVariables(Particles.ColorSprites:LinearColor=&MultiplyLinearColor(
     Linear Color=&ColorFromCurve(
      ColorCurve=DataInterface() {propertyValues {
        AlphaCurve {keys [<at 0 1 RCIM_Cubic>,<at 1 1 RCIM_Cubic>]}
        BlueCurve {keys [
          {arriveTangent=-0.2836, interpMode=RCIM_Cubic, time=0, value=0.2226}
          <at 0.1682 0.6359>
          <at 0.2618 0.12>
          <at 0.3725 0>
          <at 0.7877 0>
          <at 1 0>
         ]}
        GreenCurve {keys [
          {arriveTangent=-0.3781, interpMode=RCIM_Cubic, time=0, value=0.0541}
          <at 0.1682 0.05>
          <at 0.2618 0.1933>
          <at 0.3725 0.1667>
          <at 0.7877 0.25>
          <at 1 0.0333>
         ]}
        RedCurve {keys [
          <at 0 0.5 RCIM_Cubic>
          <at 0.1682 1>
          <at 0.2618 1>
          {arriveTangent=21.5915, leaveTangent=21.5915, time=0.3725, value=1}
          <at 0.7877 1>
          <at 1 1>
         ]}
       }}
      CurveIndex=&Multiply_Float(A=&RandomRangeFloat(Minimum=0.8, Maximum=1), B=$Particles.NormalizedAge)
     )
     Scale Factor=&MakeFloat4FromVectorAndFloat(
      "Vector (RGB)"=&VectorFromFloat(Value=1)
      "Float (Alpha)"=&FloatFromCurve(FloatCurve=DataInterface() {propertyValues {Curve {keys [
           {interpMode=RCIM_Cubic, tangentMode=RCTM_User, time=0, value=0.3}
           {arriveTangent=-0.4601, interpMode=RCIM_Cubic, leaveTangent=-0.4601, time=0.3, value=1}
           {arriveTangent=0.5752, interpMode=RCIM_Cubic, leaveTangent=0.5752, tangentMode=RCTM_User, time=0.652, value=0}
          ]}}})
     )
    )) {isSetParameters=true, module=SetVariables_#2}
   &ScaleMeshSize(Scale Factor=&VectorFromFloat(Value=&FloatFromCurve(
      FloatCurve=DataInterface() {propertyValues {Curve {keys [
          {arriveTangent=-3.8315, interpMode=RCIM_Cubic, leaveTangent=-3.8315, tangentMode=RCTM_User, time=0.0164, value=1.5}
          {arriveTangent=-0.3258, interpMode=RCIM_Cubic, leaveTangent=-0.3258, time=0.3589, value=1.4}
          {arriveTangent=0.3867, leaveTangent=0.3867, time=0.4, value=1.375}
          {arriveTangent=1.1481, interpMode=RCIM_Cubic, leaveTangent=1.1481, tangentMode=RCTM_User, time=0.6, value=0.5}
         ]}}}
      Scale Curve=2
     ))) {module=ScaleMeshSize001}
   &CameraOffset(Camera Offset Amount=-1)
  ]
  rendererClasses [&NiagaraMeshRendererProperties, &NiagaraSpriteRendererProperties]
  renderers [
   {rendererClass=&NiagaraMeshRendererProperties, rendererIndex=0}
   {rendererClass=&NiagaraSpriteRendererProperties, rendererIndex=1}
   {rendererClass=&NiagaraMeshRendererProperties, rendererIndex=2}
  ]
 }
 Cubes_CrashCenter {
  bEnabled=false
  emitterUpdateScript [
   &EmitterState()
   &SpawnRate(SpawnRate=5, Use Spawn Probability=false)
  ]
  particleSpawnScript [
   &InitializeParticle(
    Color=<rgba 1.5 1.2 3 0.2>
    Lifetime=&RandomRangeFloat(Minimum=0.5, Maximum=0.4)
    Mesh Scale Mode=::NewEnumerator1
    Mesh Uniform Scale Max=0.2
    Mesh Uniform Scale Min=0.15
    Sprite Rotation Angle=&RandomRangeFloat(Minimum=-0.2, Maximum=0.2)
    Sprite Rotation Mode=::NewEnumerator0
    Sprite Size Mode=::NewEnumerator3
    Uniform Sprite Size=&RandomRangeFloat(Minimum=15, Maximum=15)
    UsePositionOffset=false
    Write Material Random=true
   )
   &ShapeLocation(
    Handle Radius=0.1
    Large Radius=2
    MaskSpawnGroup=false
    Offset=&MakeVector(
     X=0
     Y=&MakeCustomFloatFromBool(Bool=&RandomBool(), True Float=-6.5, False Float=6.5)
     Z=0
    )
    Offset Coordinate Space=::Local
    Offset Mode=::NewEnumerator0
    Override Seed=false
    Rotation Mode=::NewEnumerator2
    Shape Origin=$Engine.Emitter.SimulationPosition
    Shape Primitive=::NewEnumerator3
    UseOwnerScale=true
    Yaw / Pitch / Roll=<0 90 90>
   ) {module=ShapeLocation001}
   &AddVelocity(
    Module.Use Falloff Distance=false
    Override Seed=false
    Velocity Mode=::NewEnumerator1
    Velocity Speed=&RandomRangeFloat(Minimum=-8, Maximum=-8)
   ) {module=AddVelocity001}
   &ApplyInitialForces(Apply Force to Position=true)
   &InitialMeshOrientation(
    Mesh Orientation Mode=::NewEnumerator7
    Rotation=&MakeVector(
     X=0
     Y=&SelectFloatFromArray(
      Array Sampling Mode=::NewEnumerator0
      Direct Array Index=?
      Float Selection Array=DataInterface() {propertyValues {FloatData [0,0.25,0.5,0.75]}}
      Random Seed=0
     )
     Z=0
    )
    Use Rotation Vector=true
   )
   &AddVelocity(Module.Use Falloff Distance=false, Override Seed=false, Velocity Mode=::NewEnumerator1) {enabled=false}
  ]
  particleUpdateScript [
   &ParticleState()
   &ShapeLocation(
    Cylinder Height=20
    Cylinder Radius=5
    MaskSpawnGroup=false
    Override Seed=false
    Rotation Mode=::NewEnumerator2
    Shape Primitive=::NewEnumerator1
    Surface Only=true
    Surface Only Band Thickness:NiagaraFloat=&FloatFromCurve(FloatCurve=DataInterface() {propertyValues {Curve {keys [{tangentMode=RCTM_User, time=0, value=0}, {tangentMode=RCTM_User, time=1, value=1}]}}})
    Use Endcaps In Surface Only Mode=false
    UseOwnerScale=true
    Yaw / Pitch / Roll=<0 90 90>
   ) {enabled=false}
   &VortexVelocity(
    Module.Use Influence Falloff=false
    Velocity Amount:NiagaraFloat=&FloatFromCurve(
     FloatCurve=DataInterface() {propertyValues {Curve {keys [{interpMode=RCIM_Cubic, tangentMode=RCTM_User, time=0, value=1}, {interpMode=RCIM_Cubic, tangentMode=RCTM_User, time=1, value=0.4}]}}}
     Scale Curve=-4
    )
    Vortex Axis:Vector3f=<0 1 0>
   )
   &ScaleVelocity(Velocity Scale=&MakeVector(
     X=0
     Y=&FloatFromCurve(
      FloatCurve=DataInterface() {propertyValues {Curve {keys [
          {arriveTangent=-3.5437, interpMode=RCIM_Cubic, leaveTangent=-3.5437, tangentMode=RCTM_User, time=0, value=2}
          {arriveTangent=-2.1889, interpMode=RCIM_Cubic, leaveTangent=-2.1889, tangentMode=RCTM_User, time=1, value=0.6}
         ]}}}
      Scale Curve=0.9
     )
     Z=0
    )) {enabled=false}
   &Drag(Drag=16, Rotational Drag=8, Use Linear Drag=true, Use Rotational Drag=false) {enabled=false}
   &SolveForcesAndVelocity(Clamp Velocity=false, Limit Acceleration=false)
   &ScaleMeshSize(Scale Factor=&VectorFromCurve(
     VectorCurve=DataInterface() {propertyValues {
       XCurve {keys [{arriveTangent=3.291, interpMode=RCIM_Cubic, leaveTangent=3.291, tangentMode=RCTM_User, time=0, value=3.4916}, <at 1 0>]}
       YCurve {keys [
         {arriveTangent=8.4978, interpMode=RCIM_Cubic, leaveTangent=8.4978, tangentMode=RCTM_User, time=0, value=3.1029}
         {interpMode=RCIM_Cubic, tangentMode=RCTM_User, time=1, value=1}
        ]}
       ZCurve {keys [
         {arriveTangent=4.1229, interpMode=RCIM_Cubic, leaveTangent=4.1229, tangentMode=RCTM_User, time=0, value=2.0895}
         {arriveTangent=-0.5079, interpMode=RCIM_Cubic, leaveTangent=-0.5079, tangentMode=RCTM_User, time=1.7048, value=0.9575}
         {arriveTangent=-0.358, interpMode=RCIM_Cubic, leaveTangent=-0.358, tangentMode=RCTM_User, time=4, value=0}
        ]}
      }}
     Scale Curve=<1.2 1.2 1.6>
    )) {enabled=false}
   &ScaleMeshSize(Scale Factor=&VectorFromFloat(Value=&FloatFromCurve(
      FloatCurve=DataInterface() {propertyValues {Curve {keys [{tangentMode=RCTM_User, time=0, value=0.5}, <at 0.2 1>, <at 0.5 3>, {tangentMode=RCTM_User, time=1, value=4}]}}}
      Scale Curve=2
     ))) {enabled=false, module=ScaleMeshSize001}
   &ScaleMeshSize(Scale Factor=&VectorFromFloat(Value=&FloatFromCurve(
      FloatCurve=DataInterface() {propertyValues {Curve {keys [{interpMode=RCIM_Cubic, tangentMode=RCTM_User, time=0.2, value=0.5}, {interpMode=RCIM_Cubic, tangentMode=RCTM_User, time=0.4, value=1}]}}}
      Scale Curve=20
     ))) {module=ScaleMeshSize002}
   &Color(Color=&ColorFromCurve(
     ColorCurve=DataInterface() {propertyValues {
       AlphaCurve {keys [<at 0 1 RCIM_Cubic>,<at 1 1 RCIM_Cubic>]}
       BlueCurve {keys [<at 0 0>,<at 0.1115 0>,<at 0.2661 0 RCIM_Cubic>,<at 0.3914 0>,<at 0.683 0>,<at 0.865 0 RCIM_Cubic>]}
       GreenCurve {keys [
         <at 0 0.1667>
         <at 0.1115 0.6667>
         {arriveTangent=0.5956, interpMode=RCIM_Cubic, leaveTangent=0.5956, time=0.2661, value=0.0833}
         <at 0.3914 0.8333>
         <at 0.683 0.8333>
         <at 0.865 0.1667 RCIM_Cubic>
        ]}
       RedCurve {keys [
         <at 0 5>
         <at 0.1115 20>
         {arriveTangent=-35.7343, interpMode=RCIM_Cubic, leaveTangent=-35.7343, time=0.2661, value=1}
         <at 0.3914 10>
         <at 0.683 10>
         <at 0.865 5 RCIM_Cubic>
        ]}
      }}
     CurveIndex=&Multiply_Float(A=&RandomRangeFloat(Minimum=0, Maximum=1), B=$Particles.NormalizedAge)
    )) {enabled=false}
   &ScaleSpriteSize(
    Uniform Curve Scale=3
    Uniform Curve Sprite Scale=DataInterface() {propertyValues {Curve {keys [
        {interpMode=RCIM_Cubic, tangentMode=RCTM_User, time=0.0075, value=0.2417}
        {arriveTangent=4.8181, interpMode=RCIM_Cubic, leaveTangent=4.8181, time=0.0434, value=1.0007}
        {arriveTangent=-2.066, interpMode=RCIM_Cubic, leaveTangent=-2.066, time=0.0924, value=0.6509}
        <at 0.3051 0.46>
        {arriveTangent=-0.4725, interpMode=RCIM_Cubic, leaveTangent=-0.4725, time=0.5, value=0.1}
        {arriveTangent=2.3255, interpMode=RCIM_Cubic, leaveTangent=2.3255, time=0.7986, value=0.2269}
        {arriveTangent=1.356, interpMode=RCIM_Cubic, leaveTangent=1.356, time=0.8986, value=1.0269}
        {arriveTangent=2.875, interpMode=RCIM_Cubic, leaveTangent=2.875, tangentMode=RCTM_User, time=1, value=0.5}
       ]}}}
   )
   &ScaleColor(
    Scale Alpha=0.45
    Scale RGB=&MakeVectorFromLinearColorRGB(Color=&ColorFromCurve(
      ColorCurve=DataInterface() {propertyValues {
        AlphaCurve {keys [<at 0 1 RCIM_Cubic>,<at 1 1 RCIM_Cubic>]}
        BlueCurve {keys [
          <at 0 0.5>
          <at 0.0854 0.2>
          {arriveTangent=3.0604, interpMode=RCIM_Cubic, leaveTangent=3.0604, time=0.18, value=0.1851}
          <at 0.6152 1.8215>
          <at 0.8353 5>
         ]}
        GreenCurve {keys [
          <at 0 0>
          <at 0.0854 0.2>
          {arriveTangent=-0.3775, interpMode=RCIM_Cubic, leaveTangent=-0.3775, time=0.18, value=0.1851}
          <at 0.6152 0>
          <at 0.8353 0.318>
         ]}
        RedCurve {keys [
          <at 0 0.25>
          <at 0.0854 2>
          {arriveTangent=1.8873, interpMode=RCIM_Cubic, leaveTangent=1.8873, time=0.18, value=2}
          <at 0.6152 3>
          <at 0.8353 4.7764>
         ]}
       }}
      CurveIndex=&Multiply_Float(A=&RandomRangeFloat(Minimum=0.5, Maximum=0.5), B=$Particles.NormalizedAge)
     ))
    ScaleA=true
    ScaleRGB=true
   )
   /SetVariables(Particles.ColorSprites:LinearColor=&MultiplyLinearColor(
     Linear Color=$Particles.Color
     Scale Factor=&MakeFloat4FromVectorAndFloat(
      "Vector (RGB)"=&VectorFromFloat(Value=0.4)
      "Float (Alpha)"=&FloatFromCurve(
       FloatCurve=DataInterface() {propertyValues {Curve {keys [
           {interpMode=RCIM_Cubic, tangentMode=RCTM_User, time=0, value=0.3}
           <at 0.3 0.2 RCIM_Cubic>
           {arriveTangent=0.5752, interpMode=RCIM_Cubic, leaveTangent=0.5752, tangentMode=RCTM_User, time=1, value=1}
          ]}}}
       Scale Curve=2
      )
     )
    )) {isSetParameters=true}
   &DynamicMaterialParameters(
    Index 0 Param 1=&FloatFromCurve(FloatCurve=DataInterface() {propertyValues {Curve {keys [
         {interpMode=RCIM_Cubic, tangentMode=RCTM_User, time=0, value=1}
         <at 0.2 0.2>
         {interpMode=RCIM_Cubic, tangentMode=RCTM_User, time=1, value=0}
        ]}}})
    Index 0 Param 3=&FloatFromCurve(
     FloatCurve=DataInterface() {propertyValues {Curve {keys [
         {arriveTangent=-5.0918, interpMode=RCIM_Cubic, leaveTangent=-5.0918, tangentMode=RCTM_User, time=0, value=1.2}
         {arriveTangent=-3.3873, interpMode=RCIM_Cubic, leaveTangent=-3.3873, tangentMode=RCTM_User, time=1, value=0}
        ]}}}
     Scale Curve=-3
    )
    Module.Param0WriteEnabled=true
    Module.Param1WriteEnabled=true
    Module.Param2WriteEnabled=true
    Module.Param3WriteEnabled=true
   )
   /SetVariables(Particles.DynamicFlare:Vector4f=&MakeVector4(X=1, Y=1, Z=0, W=0)) {isSetParameters=true}
   &CameraOffset(Camera Offset Amount=-1)
  ]
  rendererClasses [&NiagaraMeshRendererProperties, &NiagaraSpriteRendererProperties]
  renderers [
   {rendererClass=&NiagaraMeshRendererProperties, rendererIndex=0}
   {rendererClass=&NiagaraSpriteRendererProperties, rendererIndex=1}
   {rendererClass=&NiagaraSpriteRendererProperties, rendererIndex=2}
  ]
 }
 Cubes_FlowingToCenter {
  emitterUpdateScript [
   &EmitterState()
   &SpawnRate(
    Spawn Probability=&FloatFromCurve(
     FloatCurve=DataInterface() {propertyValues {Curve {keys [
         {interpMode=RCIM_Cubic, tangentMode=RCTM_User, time=0.2, value=0.5}
         {interpMode=RCIM_Cubic, tangentMode=RCTM_User, time=0.5, value=1}
         {tangentMode=RCTM_User, time=0.8, value=0.1}
        ]}}}
     CurveIndex=$Emitter.NormalizedLoopAge
    )
    SpawnRate=50
    Use Spawn Probability=true
   )
  ]
  particleSpawnScript [
   &InitializeParticle(
    Color=&MakeLinearColorFromVectorAndFloat(
     "Vector (RGB)"=&VectorFromFloat(Value=&RandomRangeFloat(Minimum=0.1, Maximum=0.3))
     "Float (Alpha)"=2
    )
    Lifetime=&RandomRangeFloat(Minimum=0.8, Maximum=1)
    Mesh Scale Max=<0.6 0.3 0.7>
    Mesh Scale Min=<0.3 0.1 0.3>
    Mesh Scale Mode=::NewEnumerator2
    Sprite Size Mode=::NewEnumerator3
    Uniform Sprite Size=&RandomRangeFloat(Minimum=5, Maximum=7)
    UsePositionOffset=false
    Write Material Random=true
   )
   &ShapeLocation(
    Handle Radius=0.2
    Large Radius=1.5
    MaskSpawnGroup=false
    Non Uniform Scale=<1 1 8>
    Offset=&MakeVector(
     X=0
     Y=&MakeCustomFloatFromBool(Bool=&RandomBool(), True Float=-12, False Float=12)
     Z=0
    )
    Offset Coordinate Space=::Local
    Offset Mode=::NewEnumerator0
    Override Seed=false
    Rotation Mode=::NewEnumerator2
    Shape Origin=$Engine.Emitter.SimulationPosition
    Shape Primitive=::NewEnumerator3
    UseOwnerScale=true
    Yaw / Pitch / Roll=<0 90 90>
   ) {module=ShapeLocation001}
   &RotateAroundPoint(Radius=2, Rotation Phase=0.1, Rotation Rate=1, Y Axis:Vector3f=<0 0 1>) {enabled=false}
   &PointAttractionForce(Attraction Strength=0.5, Kill Within Radius=false, Module.Use Falloff=false, UseAttractorOffset=false) {enabled=false}
   &AddVelocity(
    Module.Use Falloff Distance=false
    Override Seed=false
    Velocity Mode=::NewEnumerator1
    Velocity Speed=&RandomRangeFloat(Minimum=-6, Maximum=-7)
   ) {module=AddVelocity001}
   &ApplyInitialForces(Apply Force to Position=true)
   &AddVelocity(Module.Use Falloff Distance=false, Override Seed=false, Velocity Mode=::NewEnumerator1) {enabled=false}
   &InitialMeshOrientation(
    Mesh Orientation Mode=::NewEnumerator7
    Rotation=&MakeVector(X=0, Y=&RandomRangeFloat(Minimum=0, Maximum=1), Z=0)
    Use Rotation Vector=true
   ) {enabled=false}
   &InheritVelocity(Inherited Velocity Amount Scale=<0.5 0.5 0.5>, Inherited Velocity Speed Limit=3, Speed Limit=true, UseSpeedThreshold=false)
   /SetVariables(Particles.DynamicParamExtra:Vector4f=&MakeVector4(X=3, Y=0, Z=0, W=0)) {isSetParameters=true}
  ]
  particleUpdateScript [
   &ParticleState()
   &ShapeLocation(
    Cylinder Height=20
    Cylinder Radius=5
    MaskSpawnGroup=false
    Override Seed=false
    Rotation Mode=::NewEnumerator2
    Shape Primitive=::NewEnumerator1
    Surface Only=true
    Surface Only Band Thickness:NiagaraFloat=&FloatFromCurve(FloatCurve=DataInterface() {propertyValues {Curve {keys [{tangentMode=RCTM_User, time=0, value=0}, {tangentMode=RCTM_User, time=1, value=1}]}}})
    Use Endcaps In Surface Only Mode=false
    UseOwnerScale=true
    Yaw / Pitch / Roll=<0 90 90>
   ) {enabled=false}
   &VortexVelocity(
    Module.Use Influence Falloff=false
    Velocity Amount:NiagaraFloat=&FloatFromCurve(
     FloatCurve=DataInterface() {propertyValues {Curve {keys [{interpMode=RCIM_Cubic, tangentMode=RCTM_User, time=0, value=1}, {interpMode=RCIM_Cubic, tangentMode=RCTM_User, time=1, value=0}]}}}
     Scale Curve=2000
    )
    Vortex Axis:Vector3f=<0 1 0>
   )
   &ScaleVelocity(Velocity Scale=&MakeVector(
     X=&FloatFromCurve(
      FloatCurve=DataInterface() {propertyValues {Curve {keys [
          {arriveTangent=4.4437, interpMode=RCIM_Cubic, leaveTangent=4.4437, tangentMode=RCTM_User, time=0.5, value=0}
          {arriveTangent=-0.8586, interpMode=RCIM_Cubic, leaveTangent=-0.8586, tangentMode=RCTM_User, time=1, value=1}
         ]}}}
      Scale Curve=-6
     )
     Y=&FloatFromCurve(
      FloatCurve=DataInterface() {propertyValues {Curve {keys [
          {interpMode=RCIM_Cubic, tangentMode=RCTM_User, time=0, value=0.4}
          {interpMode=RCIM_Cubic, tangentMode=RCTM_User, time=0.1517, value=1.1096}
          <at 0.3922 1.2096>
          {tangentMode=RCTM_User, time=0.5922, value=0.1}
         ]}}}
      Scale Curve=3
     )
     Z=&FloatFromCurve(
      FloatCurve=DataInterface() {propertyValues {Curve {keys [
          {arriveTangent=6.7878, interpMode=RCIM_Cubic, leaveTangent=6.7878, tangentMode=RCTM_User, time=0.5, value=0}
          {arriveTangent=-1.0738, interpMode=RCIM_Cubic, leaveTangent=-1.0738, tangentMode=RCTM_User, time=1, value=1}
         ]}}}
      Scale Curve=-4
     )
    ))
   &Drag(Drag=16, Rotational Drag=8, Use Linear Drag=true, Use Rotational Drag=false) {enabled=false}
   &SolveForcesAndVelocity(Clamp Velocity=false, Limit Acceleration=false)
   &ScaleMeshSize(Scale Factor=&MakeVector(
     X=&FloatFromCurve(
      FloatCurve=DataInterface() {propertyValues {Curve {keys [
          <at 0 0.5>
          {arriveTangent=1.0382, interpMode=RCIM_Cubic, leaveTangent=1.0382, tangentMode=RCTM_User, time=0.2, value=0.2}
          {interpMode=RCIM_Cubic, tangentMode=RCTM_User, time=1, value=0.4}
         ]}}}
      Scale Curve=3
     )
     Y=&FloatFromCurve(FloatCurve=DataInterface() {propertyValues {Curve {keys [
          {arriveTangent=-5.7231, interpMode=RCIM_Cubic, leaveTangent=-5.7231, tangentMode=RCTM_User, time=0, value=1}
          {arriveTangent=11.6115, interpMode=RCIM_Cubic, leaveTangent=11.6115, time=0.1608, value=3.3179}
          <at 0.2821 4.2757>
          {arriveTangent=2.5208, interpMode=RCIM_Cubic, leaveTangent=2.5208, tangentMode=RCTM_User, time=0.39, value=3}
         ]}}})
     Z=&FloatFromCurve(
      FloatCurve=DataInterface() {propertyValues {Curve {keys [
          <at 0 0.5>
          {arriveTangent=0.9168, interpMode=RCIM_Cubic, leaveTangent=0.9168, tangentMode=RCTM_User, time=0.2, value=0.2}
          {interpMode=RCIM_Cubic, tangentMode=RCTM_User, time=1, value=0.8}
         ]}}}
      Scale Curve=3
     )
    ))
   &ScaleMeshSize(Scale Factor=&VectorFromFloat(Value=&FloatFromCurve(
      FloatCurve=DataInterface() {propertyValues {Curve {keys [
          {arriveTangent=-3.8315, interpMode=RCIM_Cubic, leaveTangent=-3.8315, tangentMode=RCTM_User, time=0.0164, value=1.5}
          {arriveTangent=1.0339, interpMode=RCIM_Cubic, leaveTangent=1.0339, time=0.3589, value=1.4}
          {arriveTangent=0.3867, leaveTangent=0.3867, time=0.5, value=3}
          {arriveTangent=1.1481, interpMode=RCIM_Cubic, leaveTangent=1.1481, tangentMode=RCTM_User, time=0.6, value=0}
         ]}}}
      Scale Curve=2
     ))) {module=ScaleMeshSize001}
   &ScaleMeshSize(Scale Factor=&Multiply_Vector(
     A=&MakeVector(
      X=&FloatFromCurve(
       FloatCurve=DataInterface() {propertyValues {Curve {keys [
           {interpMode=RCIM_Cubic, tangentMode=RCTM_User, time=0, value=0}
           {interpMode=RCIM_Cubic, tangentMode=RCTM_User, time=0.5, value=1}
           {tangentMode=RCTM_User, time=1, value=0}
          ]}}}
       Scale Curve=2
      )
      Y=1
      Z=1
     )
     B=&VectorFromFloat(Value=&FloatFromCurve(
       FloatCurve=DataInterface() {propertyValues {Curve {keys [
           {interpMode=RCIM_Cubic, tangentMode=RCTM_User, time=0.2, value=0.1}
           {arriveTangent=2.5962, interpMode=RCIM_Cubic, leaveTangent=2.5962, tangentMode=RCTM_User, time=1, value=1}
          ]}}}
       Scale Curve=6
      ))
    )) {enabled=false, module=ScaleMeshSize002}
   &Color(Color=&ColorFromCurve(
     ColorCurve=DataInterface() {propertyValues {
       AlphaCurve {keys [<at 0 1 RCIM_Cubic>,<at 1 1 RCIM_Cubic>]}
       BlueCurve {keys [<at 0 0>,<at 0.1115 0>,<at 0.2661 0 RCIM_Cubic>,<at 0.3914 0>,<at 0.683 0>,<at 0.865 0 RCIM_Cubic>]}
       GreenCurve {keys [
         <at 0 0.1667>
         <at 0.1115 0.6667>
         {arriveTangent=0.5956, interpMode=RCIM_Cubic, leaveTangent=0.5956, time=0.2661, value=0.0833}
         <at 0.3914 0.8333>
         <at 0.683 0.8333>
         <at 0.865 0.1667 RCIM_Cubic>
        ]}
       RedCurve {keys [
         <at 0 5>
         <at 0.1115 20>
         {arriveTangent=-35.7343, interpMode=RCIM_Cubic, leaveTangent=-35.7343, time=0.2661, value=1}
         <at 0.3914 10>
         <at 0.683 10>
         <at 0.865 5 RCIM_Cubic>
        ]}
      }}
     CurveIndex=&Multiply_Float(A=&RandomRangeFloat(Minimum=0, Maximum=1), B=$Particles.NormalizedAge)
    )) {enabled=false}
   &ScaleSpriteSize(
    Uniform Curve Scale=4
    Uniform Curve Sprite Scale=DataInterface() {propertyValues {Curve {keys [
        {interpMode=RCIM_Cubic, tangentMode=RCTM_User, time=0, value=0}
        {arriveTangent=6.9659, interpMode=RCIM_Cubic, leaveTangent=6.9659, time=0.0359, value=0.759}
        {arriveTangent=-0.3574, interpMode=RCIM_Cubic, leaveTangent=-0.3574, time=0.0877, value=0.611}
        <at 0.3079 0.6618>
        {arriveTangent=-1.6878, interpMode=RCIM_Cubic, leaveTangent=-1.6878, time=0.5077, value=0.5932}
        {arriveTangent=-0.1479, leaveTangent=-0.1479, time=0.7, value=0}
        {arriveTangent=-0.1667, leaveTangent=-0.1667, time=0.8, value=0}
        {arriveTangent=2.875, leaveTangent=2.875, time=1, value=0}
       ]}}}
   )
   &ScaleColor(
    Scale Alpha=&FloatFromCurve(FloatCurve=DataInterface() {propertyValues {Curve {keys [
         {tangentMode=RCTM_User, time=0, value=1}
         {arriveTangent=-0.0405, interpMode=RCIM_Cubic, leaveTangent=-0.0405, tangentMode=RCTM_User, time=0.3, value=1}
         {arriveTangent=-8.0089, interpMode=RCIM_Cubic, leaveTangent=-8.0089, tangentMode=RCTM_User, time=0.6, value=0}
        ]}}})
    Scale RGB=&MakeVectorFromLinearColorRGB(Color=&ColorFromCurve(
      ColorCurve=DataInterface() {propertyValues {
        AlphaCurve {keys [<at 0 1 RCIM_Cubic>,<at 1 1 RCIM_Cubic>]}
        BlueCurve {keys [<at 0 0>,<at 0.1115 0.02>,<at 0.4269 0>,<at 0.522 0.06>,<at 0.6222 1>,<at 1 0.5>]}
        GreenCurve {keys [<at 0 0.0333>,<at 0.1115 0.25>,<at 0.4269 0.1667>,<at 0.522 0.0967>,<at 0.6222 0>,<at 1 0>]}
        RedCurve {keys [
          <at 0 1>
          <at 0.1115 1>
          {arriveTangent=21.5915, leaveTangent=21.5915, time=0.4269, value=1}
          <at 0.522 0.5>
          <at 0.6222 0.6648>
          <at 1 0.3306>
         ]}
       }}
      CurveIndex=&Multiply_Float(A=&RandomRangeFloat(Minimum=1.5, Maximum=2), B=$Particles.NormalizedAge)
     ))
    ScaleA=true
    ScaleRGB=true
   )
   /SetVariables(Particles.ColorSprites:LinearColor=&MultiplyLinearColor(
     Linear Color=&ColorFromCurve(
      ColorCurve=DataInterface() {propertyValues {
        AlphaCurve {keys [<at 0 1 RCIM_Cubic>,<at 1 1 RCIM_Cubic>]}
        BlueCurve {keys [
          <at 0 0>
          <at 0.1115 0.3338>
          <at 0.2966 0>
          <at 0.6392 0.12>
          <at 0.6923 0>
          {interpMode=RCIM_Cubic, leaveTangent=1.9266, time=0.9853, value=0.4936}
         ]}
        GreenCurve {keys [
          <at 0 0.0333>
          <at 0.1115 0>
          <at 0.2966 0.0833>
          <at 0.6392 0.1933>
          <at 0.6923 0.0415>
          {interpMode=RCIM_Cubic, leaveTangent=-0.1606, time=0.9853, value=0}
         ]}
        RedCurve {keys [
          <at 0 1>
          <at 0.1115 1>
          {arriveTangent=21.5915, leaveTangent=21.5915, time=0.2966, value=1}
          <at 0.6392 1>
          <at 0.6923 1>
          {interpMode=RCIM_Cubic, leaveTangent=-0.8283, time=0.9853, value=1}
         ]}
       }}
      CurveIndex=&Multiply_Float(A=&RandomRangeFloat(Minimum=0.8, Maximum=1), B=$Particles.NormalizedAge)
     )
     Scale Factor=&MakeFloat4FromVectorAndFloat(
      "Vector (RGB)"=&VectorFromFloat(Value=1)
      "Float (Alpha)"=&FloatFromCurve(
       FloatCurve=DataInterface() {propertyValues {Curve {keys [
           {interpMode=RCIM_Cubic, tangentMode=RCTM_User, time=0, value=0.3}
           {arriveTangent=0.7, interpMode=RCIM_Cubic, leaveTangent=0.7, time=0.3, value=0.6}
           {arriveTangent=0.5752, interpMode=RCIM_Cubic, leaveTangent=0.5752, tangentMode=RCTM_User, time=1, value=1}
          ]}}}
       Scale Curve=0.3
      )
     )
    )) {isSetParameters=true, module=SetVariables_#2}
   &DynamicMaterialParameters(
    Index 0 Param 1=&FloatFromCurve(FloatCurve=DataInterface() {propertyValues {Curve {keys [<at 0 1>,<at 0.0894 3.7076>,<at 0.1894 4.8872>,<at 0.1894 4.7622>,<at 0.2811 1.3361>,<at 0.4895 2.419>,<at 0.577 1>,<at 0.9096 1>]}}})
    Index 0 Param 3=&FloatFromCurve(
     FloatCurve=DataInterface() {propertyValues {Curve {keys [
         {arriveTangent=-0.5377, interpMode=RCIM_Cubic, leaveTangent=-0.5377, tangentMode=RCTM_User, time=0, value=1}
         {arriveTangent=-2.3485, interpMode=RCIM_Cubic, leaveTangent=-2.3485, tangentMode=RCTM_User, time=0.164, value=1}
         {arriveTangent=-3.0281, interpMode=RCIM_Cubic, leaveTangent=-3.0281, tangentMode=RCTM_User, time=0.7085, value=0.0909}
        ]}}}
     Scale Curve=-1
    )
    Module.Param0WriteEnabled=true
    Module.Param1WriteEnabled=true
    Module.Param2WriteEnabled=true
    Module.Param3WriteEnabled=true
   )
   &CameraOffset(Camera Offset Amount=-1)
  ]
  rendererClasses [&NiagaraMeshRendererProperties, &NiagaraSpriteRendererProperties]
  renderers [
   {rendererClass=&NiagaraMeshRendererProperties, rendererIndex=0}
   {rendererClass=&NiagaraSpriteRendererProperties, rendererIndex=1}
  ]
 }
 Decal_ResetBurst {
  bEnabled=false
  emitterUpdateScript [
   &EmitterState(
    Inactive Response=::NewEnumerator1
    Life Cycle Mode=::NewEnumerator1
    Loop Behavior=::NewEnumerator1
    Loop Delay=0.3
    Loop Duration=1
    UseLoopDelay=true
   )
   &SpawnBurst_Instantaneous(
    Spawn Count=2
    Spawn Probability=&MakeCustomFloatFromBool(Bool=$User.bIsFrontEndPreview, True Float=0, False Float=1)
    Spawn Time=0.01
    Use Spawn Probability=true
   )
   &SpawnRate(SpawnRate=15, Use Spawn Probability=false) {enabled=false}
  ]
  particleSpawnScript [/Niagara/Modules/Spawn/Initialization/InitializeParticle(
    Color=<rgba 0 0 0 1>
    Module.Write Color=true
    Module.Write Lifetime=true
    Module.Write Mass=true
    Module.Write Material Random=true
    Module.Write Position=false
    Module.Write Scale=false
    Module.Write SpriteRotation=false
    Module.Write SpriteSize=true
    Sprite Size=&Vector2DFromFloat(Value=15)
   )]
  particleUpdateScript [
   &ParticleState()
   &Color(Color=&MultiplyLinearColor(
     Linear Color=<rgba 1 0.2269 0.4078 1>
     Scale Factor=&MakeFloat4FromVectorAndFloat("Vector (RGB)"=&VectorFromFloat(Value=1), "Float (Alpha)"=1)
    ))
   &ScaleColor(
    Color Value To Scale=$Particles.Color
    Scale Alpha=&FloatFromCurve(
     FloatCurve=DataInterface() {propertyValues {Curve {keys [{arriveTangent=7.9503, interpMode=RCIM_Cubic, time=0, value=0}, <at 0.3 1.25 RCIM_Cubic>, <at 1 0>]}}}
     Scale Curve=2
    )
    Scale RGB=&VectorFromFloat(Value=1)
    ScaleA=true
    ScaleRGB=true
   )
   &SkeletalMeshLocation(
    Bone / Socket Index=&ReturnParticleIDAttributes()
    Bone Sampling Mode=::NewEnumerator4
    Inherit Velocity=false
    MaskSpawnGroup=false
    Sampled Position Offset=&SelectVectorFromArray(
     Direct Array Index=&ReturnParticleIDAttributes()
     Vector Selection Array=DataInterface() {propertyValues {FloatData [<-6 0 0>, <6 0 0>]}}
    )
    Skeletal Mesh=$User.SkeletalMesh
   )
   &CameraOffset(Camera Offset Amount=20)
   /SetVariables() {isSetParameters=true}
   &DynamicMaterialParameters(
    Index 0 Param 1=&FloatFromCurve(FloatCurve=DataInterface() {propertyValues {Curve {keys [
         <at 0 -0.5 RCIM_Cubic>
         {arriveTangent=10, interpMode=RCIM_Cubic, leaveTangent=10, time=0.0692, value=0.2738}
         {arriveTangent=5.0777, interpMode=RCIM_Cubic, leaveTangent=5.0777, time=0.15, value=1}
         {interpMode=RCIM_Cubic, leaveTangent=4.2839, time=1, value=5}
        ]}}})
    Index 0 Param 3=&FloatFromCurve(
     FloatCurve=DataInterface() {propertyValues {Curve {keys [{arriveTangent=2.3217, interpMode=RCIM_Cubic, time=0, value=-0.25}, {interpMode=RCIM_Cubic, leaveTangent=0.1844, time=1, value=1.4}]}}}
     Scale Curve=0.5
    )
    Index 0 Param 4=1
    Module.Param0WriteEnabled=true
    Module.Param1WriteEnabled=true
    Module.Param2WriteEnabled=true
    Module.Param3WriteEnabled=true
   )
  ]
  rendererClasses [&NiagaraSpriteRendererProperties]
  renderers [
   {rendererClass=&NiagaraSpriteRendererProperties}
   {rendererClass=&NiagaraSpriteRendererProperties, rendererIndex=1}
  ]
 }
 LightningSprites_Burst001 {
  emitterUpdateScript [
   &EmitterState()
   &SpawnRate(
    Spawn Probability=&FloatFromCurve(
     FloatCurve=DataInterface() {propertyValues {Curve {keys [
         {interpMode=RCIM_Cubic, tangentMode=RCTM_User, time=0, value=1}
         {arriveTangent=1.6773, interpMode=RCIM_Cubic, leaveTangent=1.6773, tangentMode=RCTM_User, time=0.2, value=0.75}
         {arriveTangent=-0.0048, interpMode=RCIM_Cubic, leaveTangent=-0.0048, tangentMode=RCTM_User, time=0.5, value=0.5}
         {tangentMode=RCTM_User, time=1, value=1}
        ]}}}
     CurveIndex=$Emitter.NormalizedLoopAge
    )
    SpawnRate=30
    Use Spawn Probability=true
   )
  ]
  particleSpawnScript [
   &InitializeParticle(
    Color=<rgba 0.8317 0 2 1>
    Module.Write Color=true
    Module.Write Lifetime=true
    Module.Write Mass=true
    Module.Write Material Random=true
    Module.Write Position=false
    Module.Write Scale=false
    Module.Write SpriteRotation=true
    Module.Write SpriteSize=true
    Position=&SimulationPosition()
    Sprite Rotation=&UniformRangedFloat(Minimum=0, Maximum=360)
    Sprite Size=&UniformRangedVector2D(Minimum=<15 15>, Maximum=<20 20>)
   ) {enabled=false}
   /Niagara/Modules/Spawn/Initialization/V2/InitializeParticle(
    AdjustAlpha=true
    AdjustHue=false
    AdjustSaturation=false
    AdjustValue=true
    Alpha Scale Range:Vector2f=<0.5 1>
    Color=<rgba 0.8 0 2 1>
    Color Mode=::NewEnumerator3
    Lifetime=0.2
    Position Offset=&RandomRangeVector(Minimum=<5 5 115>, Maximum=<-5 -5 115>)
    Sprite Rotation Mode=::NewEnumerator2
    Sprite Size Mode=::NewEnumerator1
    Uniform Sprite Size Max=14
    Uniform Sprite Size Min=10
    UsePositionOffset=false
    Value Range:Vector2f=<0.5 1>
    Write Material Random=true
   ) {module=InitializeParticle001}
   /Niagara/Modules/Spawn/Location/SphereLocation_System(
    Module.Override Randomness=false
    Non Uniform Scale:Vector3f=<1 1 0>
    Sphere Radius=200
    Surface Expansion Mode:ENiagaraExpansionMode=::NewEnumerator1
    Surface Only=true
   ) {enabled=false}
   /SetVariables(Particles.RandomPosition:Vector3f=&UniformRangedVector(Minimum=<-30 -30 0>, Maximum=<30 30 200>)) {enabled=false, isSetParameters=true}
   &DynamicMaterialParameters(
    Index 0 Param 3=&FloatFromCurve(FloatCurve=DataInterface() {propertyValues {Curve {keys [<at 0 1 RCIM_Cubic>,<at 0.2 0.025 RCIM_Cubic>,<at 1 1>]}}})
    Module.Param0WriteEnabled=false
    Module.Param1WriteEnabled=false
    Module.Param2WriteEnabled=true
    Module.Param3WriteEnabled=false
   )
   /SetVariables(Particles.AirGlow_Color:LinearColor=<rgba 0.8317 0 2 2>, Particles.AirGlow_Size:Vector2f=<30 30>) {isSetParameters=true}
  ]
  particleUpdateScript [
   &ParticleState()
   &SkeletalMeshLocation(Bone / Socket Index=2, Bone Sampling Mode=::NewEnumerator4, MaskSpawnGroup=false, Skeletal Mesh=$User.SkeletalMesh) {module=SkeletalMeshLocation001}
   &SkeletalMeshLocation(
    Bone Sampling Mode=::NewEnumerator4
    Inherit Velocity=false
    MaskSpawnGroup=false
    Sampled Position Offset=&RandomRangeVector(Minimum=<-5 -5 110>, Maximum=<5 5 110>)
    Skeletal Mesh=$User.SkeletalMesh
   ) {enabled=false}
   /SetVariables(Particles.AirGlow_Position:NiagaraPosition=$Particles.Position) {isSetParameters=true}
   /SetVariables(Particles.SpriteFacing=&RotateVectorByQuaternion(Vector To Rotate=<1 0 0>, Quaternion=$Output.SkeletalMeshLocation.SampledRotation)) {isSetParameters=true}
   &DynamicMaterialParameters(
    Index 0 Param 3=&FloatFromCurve(FloatCurve=DataInterface() {propertyValues {Curve {keys [<at 0 1 RCIM_Cubic>,<at 0.2 0.025 RCIM_Cubic>,<at 1 1>]}}})
    Module.Param0WriteEnabled=false
    Module.Param1WriteEnabled=false
    Module.Param2WriteEnabled=true
    Module.Param3WriteEnabled=false
   ) {module=DynamicMaterialParameters001}
   &ScaleColor(
    Scale Alpha=&FloatFromCurve(FloatCurve=DataInterface() {propertyValues {Curve {keys [<at 0.75 1 RCIM_Cubic>, {arriveTangent=-10.3917, interpMode=RCIM_Cubic, leaveTangent=-10.3917, tangentMode=RCTM_User, time=1, value=0}]}}})
    Scale RGB=&Multiply_VectorByFloat(
     Vector=&MakeVectorFromLinearColorRGB(Color=&ColorFromCurve(ColorCurve=DataInterface() {propertyValues {
         AlphaCurve {keys []}
         BlueCurve {keys [<at 0 1>,<at 0.15 10 RCIM_Cubic>,<at 1 1 RCIM_Cubic>]}
         GreenCurve {keys [<at 0 1>, {arriveTangent=-0.1823, interpMode=RCIM_Cubic, leaveTangent=-0.1823, time=0.15, value=9.045}, <at 1 0.8177 RCIM_Cubic>]}
         RedCurve {keys [<at 0 1>, {arriveTangent=-0.1812, interpMode=RCIM_Cubic, leaveTangent=-0.1812, time=0.15, value=8.5417}, <at 1 0.8188 RCIM_Cubic>]}
        }}))
     Float=0.3
    )
    ScaleA=true
    ScaleRGB=true
   )
   /Niagara/Modules/Update/SubUV/SubUVAnimation(Number Of Frames=8, Random SubUV Lookup Index:NiagaraFloat=$Particles.NormalizedAge, SubUV Animation Mode=::NewEnumerator1)
   &ScaleSpriteSize(
    Uniform Curve Scale=2
    Uniform Curve Sprite Scale=DataInterface() {propertyValues {Curve {keys [{tangentMode=RCTM_User, time=0, value=0.5}, {tangentMode=RCTM_User, time=1, value=1}]}}}
   )
   &CameraOffset(Camera Offset Amount=-5)
  ]
  rendererClasses [&NiagaraSpriteRendererProperties]
  renderers [
   {rendererClass=&NiagaraSpriteRendererProperties}
   {rendererClass=&NiagaraSpriteRendererProperties, rendererIndex=1}
  ]
 }
 MeshCube01 {
  emitterUpdateScript [
   &EmitterState(
    Inactive Response=::NewEnumerator1
    Life Cycle Mode=::NewEnumerator1
    Loop Behavior=::NewEnumerator1
    Loop Delay=0.8
    Loop Duration Mode=::NewEnumerator1
    UseLoopDelay=true
   )
   &SpawnBurst_Instantaneous(Use Spawn Probability=false, UseLoopCountLimit=false) {module=SpawnBurst_Instantaneous001}
   &SpawnRate(SpawnRate=2, Use Spawn Probability=false) {enabled=false}
   &SpawnBurst_Instantaneous(Use Spawn Probability=false, UseLoopCountLimit=false) {enabled=false}
  ]
  particleSpawnScript [
   &InitializeParticle(
    Color=<rgba 0.4063 0.0924 1 1>
    Lifetime=&RandomRangeFloat(Minimum=0.2, Maximum=0.3)
    Mesh Scale Mode=::NewEnumerator3
    UsePositionOffset=false
    Write Material Random=true
   )
   &InitialMeshOrientation(Orientation Vector=<0 1 0>, Rotation Coordinate Space=::NewEnumerator0, Use Orientation Vector=true, Use Rotation Vector=true)
   &InheritVelocity(Speed Limit=true, UseSpeedThreshold=false)
  ]
  particleUpdateScript [
   /SetVariables(Particles.NormalizedAgeAlt=1) {enabled=false, isSetParameters=true}
   &ParticleState(Kill Particles When Lifetime Has Elapsed=false, Loop Particles Lifetime=true)
   &PendulumSetup(
    Module.Enable Spring Driver=true
    Pendulum Length=1.5
    Pendulum Pivot Offset:Vector3f=<0 0 2>
    Rigid Pendulum=false
    "Spring Driven Constraint (Tightness)"=2
   )
   &SolveForcesAndVelocity(Clamp Velocity=false, Limit Acceleration=false)
   &PendulumConstraint()
   &CalculateAccurateVelocity()
   &UpdateMeshOrientation(
    Rotation Coordinate Space=::NewEnumerator2
    Rotation Rate=0.2
    Rotation Vector=&RotateVectorByQuaternion(Vector To Rotate=<0 -1 0>, Quaternion=<0 0 0 1>)
   )
   &ScaleColor(
    Scale RGB=&VectorFromFloat(Value=&FloatFromCurve(
      FloatCurve=DataInterface() {propertyValues {Curve {keys [{interpMode=RCIM_Cubic, tangentMode=RCTM_User, time=0, value=1.5}, {interpMode=RCIM_Cubic, tangentMode=RCTM_User, time=1, value=1}]}}}
      Scale Curve=0.5
     ))
    ScaleA=true
    ScaleRGB=true
   )
   /SetVariables(Particles.MeshCubeSize:Vector3f=&Multiply_Vector(
     A=&VectorFromFloat(Value=0.7)
     B=&VectorFromFloat(Value=&FloatFromCurve(
       FloatCurve=DataInterface() {propertyValues {Curve {keys [
           {interpMode=RCIM_Cubic, tangentMode=RCTM_User, time=0, value=1}
           {arriveTangent=0.2357, interpMode=RCIM_Cubic, leaveTangent=0.2357, tangentMode=RCTM_User, time=0.2, value=0.9625}
           <at 0.4 1.15>
           {tangentMode=RCTM_User, time=0.5322, value=1}
          ]}}}
       CurveIndex=&Multiply_Float(
        A=$Particles.NormalizedAge
        B=&Waveform("Waveform [1]"=::NewEnumerator0, "[1] Frequency"=1, "[1] Amplitude Min/Max"=<0 1>, "[1] Bias"=0.5, "[1] Pulse Duration"=0.8)
       )
       Scale Curve=1.3
      ))
    )) {isSetParameters=true}
   /SetVariables(
    Particles.MeshOutlineCubeColor:LinearColor=&HueShiftLinearColor(
     Hue=0
     Saturation=&MaskFloatBySpawnGroup(Masked Float=2, Passthrough Float=4)
     Value=&MaskFloatBySpawnGroup(Masked Float=0.2, Passthrough Float=0.4)
     Linear Color=&MultiplyLinearColor(
      Linear Color=&LerpLinearColors(
       StartColor=<rgba 4.8 1.6 8 1>
       EndColor=<rgba 1 0.8333 0.8 1>
       LerpFactor=&FloatFromCurve(FloatCurve=DataInterface() {propertyValues {Curve {keys [{interpMode=RCIM_Cubic, tangentMode=RCTM_User, time=0, value=0}, {interpMode=RCIM_Cubic, tangentMode=RCTM_User, time=1, value=1}]}}})
      )
      Scale Factor=&MakeVector4_FromFloat(Float=&Multiply_Float(
        A=$Particles.NormalizedAge
        B=&Waveform(
         Waveform Count=::NewEnumerator1
         Global Amplitude Scale=2
         "Waveform [1]"=::NewEnumerator0
         "[1] Phase Scale"=0.5
         "[1] Frequency"=1
         "[1] Amplitude Min/Max"=<0 1>
         "[1] Bias"=0.2
         "[2] Average"=false
         "[2] Input Position"=0
         "[2] Phase"=$Emitter.Age
         "[2] Phase Scale"=1
         "[2] Compound Frequency"=2
         "[2] Frequency"=2
         "[2] Amplitude Min/Max"=<0 1>
         "[2] Scale"=0.5
         "[2] Bias"=0
         "[2] Exponent"=1
         "[2] Pulse Duration"=0.1
         "[2] Clip Triangle Min/Max"=<0 0>
         "[2] Random Seed"=0
        )
       ))
     )
    )
    Particles.MeshOutlineSize:Vector3f=&Multiply_Vector(
     A=<1.02 0.8 0.9>
     B=&VectorFromFloat(Value=&FloatFromCurve(
       FloatCurve=DataInterface() {propertyValues {Curve {keys [
           {interpMode=RCIM_Cubic, tangentMode=RCTM_User, time=0, value=0.8}
           {arriveTangent=0.2357, interpMode=RCIM_Cubic, leaveTangent=0.2357, tangentMode=RCTM_User, time=0.1, value=1}
           <at 0.4 0.58>
           {tangentMode=RCTM_User, time=0.5322, value=0}
          ]}}}
       CurveIndex=&Multiply_Float(
        A=$Particles.NormalizedAge
        B=&Waveform("Waveform [1]"=::NewEnumerator0, "[1] Frequency"=1, "[1] Amplitude Min/Max"=<0 1>, "[1] Bias"=0.5, "[1] Pulse Duration"=0.8)
       )
       Scale Curve=2
      ))
    )
   ) {enabled=false, isSetParameters=true}
   /SetVariables(Particles.MeshOutlineCubeColor:LinearColor=<rgba 5.7009 0 10 1>, Particles.MeshOutlineSize:Vector3f=<1 1 1>) {isSetParameters=true}
   &DynamicMaterialParameters(
    Index 0 Param 1=&FloatFromCurve(
     FloatCurve=DataInterface() {propertyValues {Curve {keys [
         {interpMode=RCIM_Cubic, tangentMode=RCTM_User, time=0.1, value=0.2}
         {arriveTangent=0.3333, interpMode=RCIM_Cubic, leaveTangent=0.3333, time=0.2, value=5}
         {interpMode=RCIM_Cubic, tangentMode=RCTM_User, time=0.4, value=0.1}
         <at 0.5 0.2 RCIM_Cubic>
        ]}}}
     CurveIndex=&Add_Float(A=$Emitter.Age, B=&Multiply_Float(A=$Emitter.CurrentLoopDelay, B=-1))
    )
    Module.Param0WriteEnabled=true
    Module.Param1WriteEnabled=true
    Module.Param2WriteEnabled=true
    Module.Param3WriteEnabled=true
   ) {enabled=false}
   &ScaleMeshSize(Scale Factor=&VectorFromFloat(Value=&FloatFromCurve(
      FloatCurve=DataInterface() {propertyValues {Curve {keys [
          {interpMode=RCIM_Cubic, tangentMode=RCTM_User, time=0, value=0.2}
          {interpMode=RCIM_Cubic, tangentMode=RCTM_User, time=0.5, value=0.8}
          <at 0.8 0.975 RCIM_Cubic>
         ]}}}
      CurveIndex=&Add_Float(A=$Emitter.Age, B=&Multiply_Float(A=$Emitter.CurrentLoopDelay, B=-1))
     )))
  ]
  rendererClasses [&NiagaraMeshRendererProperties]
  renderers [{rendererClass=&NiagaraMeshRendererProperties, rendererIndex=0}]
 }
 Pulse_Cubes_Grid {
  bEnabled=false
  emitterUpdateScript [
   &EmitterState(Life Cycle Mode=::NewEnumerator1, Loop Behavior=::NewEnumerator1, Loop Delay=1.2, Loop Duration=1, UseLoopDelay=true)
   &SpawnBurst_Instantaneous(Spawn Count=10, Use Spawn Probability=false, UseLoopCountLimit=false) {enabled=false}
   /Niagara/Modules/Spawn/Location/SpawnParticlesInGrid(X Count=3, Y Count=3, Z Count=3)
  ]
  particleSpawnScript [
   &InitializeParticle(
    Color Mode=::NewEnumerator0
    Lifetime=&RandomRangeFloat(Minimum=1, Maximum=1.4)
    Mesh Scale Mode=::NewEnumerator3
    Mesh Uniform Scale=1.2
    Sprite Size Mode=::NewEnumerator3
    Uniform Sprite Size=15
    UsePositionOffset=false
    Write Material Random=true
   )
   &SampleParticlesFromOtherEmitter() {enabled=false}
   /Niagara/Modules/Spawn/Location/V2/GridLocation(Module.Jitter Inside Grid Cells=false, Module.Override Random Seed=false, Offset:Vector3f=<0 0.5 0>, XYZ Dimensions:Vector3f=<3 3 3>)
   /SetVariables(Particles.InitialPositionVector:NiagaraPosition=&self:Subtract_Position(Position 1:NiagaraPosition=$Particles.Position, Position 2:NiagaraPosition=$Engine.Owner.Position)) {isSetParameters=true}
  ]
  particleUpdateScript [
   &ParticleState()
   &DynamicMaterialParameters(
    Index 0 Param 1=&FloatFromCurve(FloatCurve=DataInterface() {propertyValues {Curve {keys [
         {arriveTangent=1.0361, interpMode=RCIM_Cubic, leaveTangent=1.0361, tangentMode=RCTM_User, time=0, value=0.2}
         <at 0.3815 0.7344>
         {arriveTangent=-2.5249, interpMode=RCIM_Cubic, leaveTangent=-2.5249, tangentMode=RCTM_User, time=0.4222, value=2.2344}
         <at 0.4815 0.4>
         {arriveTangent=-0.0405, interpMode=RCIM_Cubic, leaveTangent=-0.0405, tangentMode=RCTM_User, time=1, value=0.2}
        ]}}})
    Index 0 Param 3=-0.3545
    Module.Param0WriteEnabled=true
    Module.Param1WriteEnabled=true
    Module.Param2WriteEnabled=true
    Module.Param3WriteEnabled=true
   )
   &self:AttributeReaderMeshOrientation(New Particle Attribute Reader=DataInterface() {propertyValues {EmitterBinding {bindingMode=Other, emitterName=MeshCube01}}}) {module=AttributeReaderMeshOrientation}
   /SetVariables(Particles.Position=&self:Add_Vector_To_Position(
     Default:Vector3f=&RotateVectorByQuaternion(
      Vector To Rotate=&ConvertPositionToVector(Input Position=$Particles.InitialPositionVector, Convert To Absolute Large World Vector=false)
      Quaternion=$Transient.SampledOrientation
     )
     New Position:NiagaraPosition=$Engine.Owner.Position
    )) {isSetParameters=true}
   /SetVariables(Particles.Position_Part1:NiagaraPosition=&ConvertVectorToPosition(
     Input Position=&Multiply_Vector(A=<1 1 1>, B=$Output.GridLocation.GridSpacing)
     Vector Is In Absolute Large World Space:NiagaraBool=false
    )) {enabled=false, isSetParameters=true}
   &ScaleMeshSize(Scale Factor=&VectorFromFloat(Value=&FloatFromCurve(
      FloatCurve=DataInterface() {propertyValues {Curve {keys [
          {arriveTangent=9.0725, interpMode=RCIM_Cubic, leaveTangent=9.0725, tangentMode=RCTM_User, time=0.3, value=0}
          {arriveTangent=-2.5249, interpMode=RCIM_Cubic, leaveTangent=-2.5249, tangentMode=RCTM_User, time=0.5, value=1}
          {arriveTangent=-0.0405, interpMode=RCIM_Cubic, leaveTangent=-0.0405, tangentMode=RCTM_User, time=0.6, value=0.5}
          <at 1 0>
         ]}}}
      Scale Curve=2
     )))
   /SetVariables(Particles.ColorSprites:LinearColor=&MultiplyLinearColor(
     Linear Color=&ColorFromCurve(ColorCurve=DataInterface() {propertyValues {
        AlphaCurve {keys [<at 0 1 RCIM_Cubic>,<at 1 1 RCIM_Cubic>]}
        BlueCurve {keys [
          <at 0 0>
          <at 0.1115 0>
          <at 0.4269 0>
          <at 0.5769 0.12>
          {arriveTangent=2.8662, interpMode=RCIM_Cubic, leaveTangent=2.8662, time=0.6392, value=0.2226}
          <at 0.7569 0.6359>
         ]}
        GreenCurve {keys [
          <at 0 0.0333>
          <at 0.1115 0.25>
          <at 0.4269 0.1667>
          <at 0.5769 0.1933>
          {arriveTangent=-0.7964, interpMode=RCIM_Cubic, leaveTangent=-0.7964, time=0.6392, value=0.0541}
          <at 0.7569 0.05>
         ]}
        RedCurve {keys [
          <at 0 1>
          <at 0.1115 1>
          {arriveTangent=21.5915, leaveTangent=21.5915, time=0.4269, value=1}
          <at 0.5769 1>
          <at 0.6392 0.5 RCIM_Cubic>
          <at 0.7569 1>
         ]}
       }})
     Scale Factor=&MakeFloat4FromVectorAndFloat("Vector (RGB)"=&VectorFromFloat(Value=0.5), "Float (Alpha)"=1)
    )) {isSetParameters=true}
   &ScaleColor(
    Scale RGB=&MakeVectorFromLinearColorRGB(Color=&ColorFromCurve(
      ColorCurve=DataInterface() {propertyValues {
        AlphaCurve {keys [<at 0 1 RCIM_Cubic>,<at 1 1 RCIM_Cubic>]}
        BlueCurve {keys [<at 0.0268 1>, <at 0.1852 0.1506>, {interpMode=RCIM_Cubic, leaveTangent=-1.7519, time=0.4702, value=0.4232}]}
        GreenCurve {keys [<at 0.0268 1>, <at 0.1852 0.4777>, {interpMode=RCIM_Cubic, leaveTangent=-1.0772, time=0.4702, value=0.0379}]}
        RedCurve {keys [<at 0.0268 1>,<at 0.1852 1>,<at 0.4702 0.5 RCIM_Cubic>]}
       }}
      CurveIndex=&Multiply_Float(A=&RandomRangeFloat(Minimum=1, Maximum=1), B=$Particles.NormalizedAge)
     ))
    ScaleA=true
    ScaleRGB=true
   )
   &ScaleSpriteSize(Uniform Curve Sprite Scale=DataInterface() {propertyValues {Curve {keys [
        {interpMode=RCIM_Cubic, tangentMode=RCTM_User, time=0, value=0}
        {interpMode=RCIM_Cubic, tangentMode=RCTM_User, time=0.5, value=1}
        {tangentMode=RCTM_User, time=1, value=0}
       ]}}})
  ]
  rendererClasses [&NiagaraMeshRendererProperties, &NiagaraSpriteRendererProperties]
  renderers [
   {rendererClass=&NiagaraMeshRendererProperties, rendererIndex=0}
   {rendererClass=&NiagaraSpriteRendererProperties, rendererIndex=1}
  ]
 }
 Ripple {
  emitterUpdateScript [
   &EmitterState(Life Cycle Mode=::NewEnumerator1, Loop Behavior=::NewEnumerator1, Loop Delay=0.4, Loop Duration=0.3, UseLoopDelay=true)
   &SpawnRate(SpawnRate=30, Use Spawn Probability=false)
  ]
  particleSpawnScript [
   &InitializeParticle(Color Mode=::NewEnumerator0, Mesh Scale=<10 2 3>, Mesh Scale Mode=::NewEnumerator4, UsePositionOffset=false, Write Material Random=true)
   &BoxLocation(
    Box Offset=&MakeVector(
     X=0
     Y=&FloatFromCurve(
      FloatCurve=DataInterface() {propertyValues {Curve {keys [<at 0 0>,<at 1 1>]}}}
      CurveIndex=&Add_Float(A=$Emitter.Age, B=&Multiply_Float(A=$Emitter.CurrentLoopDelay, B=-1))
      Scale Curve=&MakeCustomFloatFromBool(
       Bool=&SetBoolByIntComparison(
        A=&ModuloInt(A=&ReturnParticleIDAttributes(), B=2)
        B=0
        Comparison Type=::NewEnumerator4
       )
       True Float=&Divide_Float(A=35, B=1)
       False Float=-35
      )
     )
     Z=0
    )
    Box Size=<0 0 0>
    MaskSpawnGroup=false
    Module.Override Seed=false
    Rotate Box=false
    Surface Only=false
   )
  ]
  particleUpdateScript [
   &ParticleState()
   &ShapeLocation(
    Handle Radius=0.2
    Large Radius=10
    MaskSpawnGroup=false
    Offset=&MakeVector(
     X=0
     Y=&MakeCustomFloatFromBool(Bool=&RandomBool(), True Float=-2, False Float=2)
     Z=0
    )
    Offset Coordinate Space=::Local
    Offset Mode=::NewEnumerator0
    Override Seed=false
    Rotation Mode=::NewEnumerator2
    Shape Primitive=::NewEnumerator3
    UseOwnerScale=true
    Yaw / Pitch / Roll=<0 90 90>
   ) {enabled=false, module=ShapeLocation001}
   &DynamicMaterialParameters(
    Index 0 Param 1=&FloatFromCurve(
     FloatCurve=DataInterface() {propertyValues {Curve {keys [<at 0 0.1>,<at 0.0833 1>,<at 0.1534 3.5031>,<at 0.2216 8>,<at 0.288 3.0177>,<at 0.4077 2.474>,<at 0.6 0.1>,<at 0.7 3>,<at 0.8 0.1>]}}}
     Scale Curve=0.5
    )
    Index 0 Param 3=&FloatFromCurve(
     FloatCurve=DataInterface() {propertyValues {Curve {keys [
         {arriveTangent=-0.5377, interpMode=RCIM_Cubic, leaveTangent=-0.5377, tangentMode=RCTM_User, time=0, value=1}
         {arriveTangent=-2.3485, interpMode=RCIM_Cubic, leaveTangent=-2.3485, tangentMode=RCTM_User, time=0.7424, value=0.55}
         {arriveTangent=-3.0281, interpMode=RCIM_Cubic, leaveTangent=-3.0281, tangentMode=RCTM_User, time=1, value=0}
        ]}}}
     Scale Curve=-1
    )
    Module.Param0WriteEnabled=true
    Module.Param1WriteEnabled=true
    Module.Param2WriteEnabled=true
    Module.Param3WriteEnabled=true
   )
   &ScaleMeshSize(Scale Factor=&VectorFromCurve(VectorCurve=DataInterface() {propertyValues {
       XCurve {keys [
         <at 0.0011 0.1 RCIM_Cubic>
         {arriveTangent=2.9148, interpMode=RCIM_Cubic, leaveTangent=2.9148, time=0.096, value=0.2692}
         {arriveTangent=0.5684, interpMode=RCIM_Cubic, leaveTangent=0.5684, time=0.2418, value=0.8016}
         {arriveTangent=-0.9142, interpMode=RCIM_Cubic, leaveTangent=-0.9142, time=0.3619, value=0.4203}
         <at 1.0093 0.1 RCIM_Cubic>
        ]}
       YCurve {keys [<at 0 0.1>,<at 0.0797 0.2>,<at 0.2797 0.6>,<at 0.4797 0.2>,<at 1 0.1>]}
       ZCurve {keys [
         <at -0.0064 0.1 RCIM_Cubic>
         {arriveTangent=2.6181, interpMode=RCIM_Cubic, leaveTangent=2.6181, time=0.2046, value=0.5017}
         {arriveTangent=-0.3119, interpMode=RCIM_Cubic, leaveTangent=-0.3119, time=0.3504, value=1.0341}
         {arriveTangent=-1.4103, interpMode=RCIM_Cubic, leaveTangent=-1.4103, time=0.5006, value=0.4094}
         <at 1.0128 0.1 RCIM_Cubic>
        ]}
      }}))
   &ScaleMeshSize(Scale Factor=&MakeVector(
     X=&FloatFromCurve(
      FloatCurve=DataInterface() {propertyValues {Curve {keys [{interpMode=RCIM_Cubic, tangentMode=RCTM_User, time=0, value=1}, {interpMode=RCIM_Cubic, tangentMode=RCTM_User, time=1, value=0.7}]}}}
      CurveIndex=&ReturnNormalizedExecIndex()
     )
     Y=1
     Z=&FloatFromCurve(
      FloatCurve=DataInterface() {propertyValues {Curve {keys [{interpMode=RCIM_Cubic, tangentMode=RCTM_User, time=0, value=1}, {interpMode=RCIM_Cubic, tangentMode=RCTM_User, time=1, value=0.7}]}}}
      CurveIndex=&ReturnNormalizedExecIndex()
     )
    )) {module=ScaleMeshSize001}
   &ScaleColor(
    Scale RGB=&MakeVectorFromLinearColorRGB(Color=&ColorFromCurve(
      ColorCurve=DataInterface() {propertyValues {
        AlphaCurve {keys [<at 0 1 RCIM_Cubic>,<at 1 1 RCIM_Cubic>]}
        BlueCurve {keys [
          <at 0 0>
          <at 0.1115 0>
          <at 0.4269 0>
          <at 0.5769 0.01>
          {interpMode=RCIM_Cubic, leaveTangent=3.4773, time=0.7877, value=0.2226}
          <at 0.8949 0.6359>
         ]}
        GreenCurve {keys [
          <at 0 0.0333>
          <at 0.1115 0.25>
          <at 0.4269 0.1667>
          <at 0.5769 0.0925>
          {interpMode=RCIM_Cubic, leaveTangent=-0.2361, time=0.7877, value=0.0541}
          <at 0.8949 0.05>
         ]}
        RedCurve {keys [
          <at 0 1>
          <at 0.1115 1>
          {arriveTangent=21.5915, leaveTangent=21.5915, time=0.4269, value=1}
          <at 0.5769 1>
          <at 0.7877 0.5 RCIM_Cubic>
          <at 0.8949 1>
         ]}
       }}
      CurveIndex=&Add_Float(A=$Emitter.Age, B=&Multiply_Float(A=$Emitter.CurrentLoopDelay, B=-1))
     ))
    ScaleA=true
    ScaleRGB=true
   )
   &UpdateMeshOrientation(Rotation Rate=-2, Rotation Vector=<0 1 0>)
  ]
  rendererClasses [&NiagaraMeshRendererProperties]
  renderers [{rendererClass=&NiagaraMeshRendererProperties, rendererIndex=0}]
 }
 SpriteHandGlow {
  emitterUpdateScript [
   &EmitterState(Life Cycle Mode=::NewEnumerator1, Loop Behavior=::NewEnumerator1, Loop Delay=0.8, Loop Duration Mode=::NewEnumerator1, UseLoopDelay=true)
   &SpawnBurst_Instantaneous(
    Spawn Count=&SelectIntFromArray(
     Direct Array Index=&MakeCustomIntFromBool(Bool=$User.bIsFrontEnd, True INT=1, False INT=0)
     Int Selection Array=DataInterface() {propertyValues {IntData [2,0]}}
    )
    Use Spawn Probability=false
    UseLoopCountLimit=false
   )
  ]
  particleSpawnScript [
   &FN_SetSpawnGroup(Number of Spawn Groups=2)
   &InitializeParticle(
    Color=<rgba 0.2 0.0121 0.1956 20>
    Lifetime=&RandomRangeFloat(Minimum=0.8, Maximum=1)
    Sprite Size Mode=::NewEnumerator3
    Uniform Sprite Size=40
    UsePositionOffset=true
    Write Material Random=true
   )
   &DynamicMaterialParameters(
    Index 0 Param 1=0.2
    Index 0 Param 2=&RandomRangeFloat(Minimum=0.5, Maximum=1)
    Index 0 Param 3=1
    Index 0 Param 4=1
    Module.Param0WriteEnabled=true
    Module.Param1WriteEnabled=true
    Module.Param2WriteEnabled=true
    Module.Param3WriteEnabled=true
   )
  ]
  particleUpdateScript [
   &ParticleState(
    Kill Particles When Lifetime Has Elapsed=false
    Let Infinitely Lived Particles Die When Emitter Deactivates=true
    Loop Particles Lifetime=true
   )
   &ScaleColor(
    Scale Alpha=&FloatFromCurve(FloatCurve=DataInterface() {propertyValues {Curve {keys [
         {interpMode=RCIM_Cubic, tangentMode=RCTM_User, time=0, value=0}
         {interpMode=RCIM_Cubic, tangentMode=RCTM_User, time=0.5, value=1}
         {tangentMode=RCTM_User, time=1, value=0}
        ]}}})
    ScaleA=true
    ScaleRGB=false
   )
   /SetVariables(Particles.Renderer2_Size:Vector2f=&Multiply_Vector2D_ByFloat(Float=0.8, Vector2D=$Particles.SpriteSize)) {isSetParameters=true}
  ]
  rendererClasses [&NiagaraSpriteRendererProperties]
  renderers [
   {rendererClass=&NiagaraSpriteRendererProperties}
   {rendererClass=&NiagaraSpriteRendererProperties, rendererIndex=1}
   {rendererClass=&NiagaraSpriteRendererProperties, rendererIndex=2}
  ]
 }
 Sprites_Low {
  emitterUpdateScript [
   &EmitterState(Life Cycle Mode=::NewEnumerator1, Loop Behavior=::NewEnumerator1, Loop Duration Mode=::NewEnumerator1, UseLoopDelay=false)
   &SpawnBurst_Instantaneous(Spawn Count=2, Use Spawn Probability=false, UseLoopCountLimit=false)
   &SpawnBurst_Instantaneous(Spawn Group=1, Use Spawn Probability=false, UseLoopCountLimit=false) {enabled=false, module=SpawnBurst_Instantaneous001}
  ]
  particleSpawnScript [
   &PartitionParticles()
   &InitializeParticle(Color=<rgba 1 0.4333 0.15 1>, Sprite Size Mode=::NewEnumerator3, Uniform Sprite Size=7, UsePositionOffset=true, Write Material Random=true)
  ]
  particleUpdateScript [
   &SystemLocation(MaskSpawnGroup=false) {enabled=false}
   &ParticleState(
    Kill Particles When Lifetime Has Elapsed=false
    Let Infinitely Lived Particles Die When Emitter Deactivates=true
    Loop Particles Lifetime=true
   )
   &SpriteFacingAndAlignment(
    Facing Coordinate Space=::Local
    Sprite Facing=&SelectVectorFromArray(
     Direct Array Index=$Particles.PartitionParticles.Partition
     Vector Selection Array=DataInterface() {propertyValues {FloatData [<0 -1 0>, <0 1 0>, <0 0 0>]}}
    )
    WriteAlignment=false
    WriteFacing=true
   )
   &SkeletalMeshLocation(
    Bone / Socket Index=$Particles.PartitionParticles.Partition
    Bone Sampling Mode=::NewEnumerator4
    MaskSpawnGroup=false
    Orientation Sampling=::NewEnumerator1
    Position Sampling=::NewEnumerator2
    Sampled Position Offset=<-0.02 0 0>
    Skeletal Mesh=$User.SkeletalMesh
   )
   &SkeletalMeshLocation(
    Bone / Socket Index=$Particles.PartitionParticles.Partition
    Bone Sampling Mode=::NewEnumerator4
    MaskSpawnGroup=false
    Orientation Sampling=::NewEnumerator1
    Position Sampling=::NewEnumerator2
    Sampled Position Offset=<-2 0 0>
    Skeletal Mesh=$User.SkeletalMesh
   ) {module=SkeletalMeshLocation001}
   &ScaleColor(
    Scale Alpha=&Waveform(
     Waveform Count=::NewEnumerator1
     Global Amplitude Scale=0.1
     "Waveform [1]"=::NewEnumerator5
     "[1] Frequency"=5
     "[1] Amplitude Min/Max"=<0.3 1>
     "[1] Pulse Duration"=0.5
     "[2] Average"=false
     "Waveform [2]"=::NewEnumerator5
     "[2] Input Position"=0
     "[2] Phase"=$Emitter.Age
     "[2] Phase Scale"=0.3
     "[2] Compound Frequency"=2
     "[2] Frequency"=5
     "[2] Amplitude Min/Max"=<0.5 1>
     "[2] Scale"=1
     "[2] Bias"=0
     "[2] Exponent"=1
     "[2] Pulse Duration"=0.6
     "[2] Clip Triangle Min/Max"=<0 0>
     "[2] Random Seed"=0
    )
    ScaleA=true
    ScaleRGB=true
   )
   /SetVariables(
    Particles.ColorLight:LinearColor=&MultiplyLinearColor(
     Linear Color=$Particles.Color
     Scale Factor=&MakeFloat4FromVectorAndFloat(
      "Vector (RGB)"=&VectorFromFloat(Value=&RandomRangeFloat(Minimum=1, Maximum=1))
      "Float (Alpha)"=1
     )
    )
    Particles.ColorSprites:LinearColor=&MultiplyLinearColor(
     Linear Color=$Particles.Color
     Scale Factor=&MakeFloat4FromVectorAndFloat("Vector (RGB)"=&VectorFromFloat(Value=2), "Float (Alpha)"=1)
    )
    Particles.LightRadius=&RandomRangeFloat(Minimum=80, Maximum=90, Evaluation Type=::NewEnumerator1)
    Particles.Position_Part1:NiagaraPosition=&ConvertVectorToPosition(
     Input Position=Linked() {linkedVariable {name=Particles.SkeletalMeshLocation.SampledPosition, type=NiagaraPosition}}
     Vector Is In Absolute Large World Space:NiagaraBool=true
    )
    Particles.Position_Part2:NiagaraPosition=&ConvertVectorToPosition(
     Input Position=Linked() {linkedVariable {name=Particles.SkeletalMeshLocation001.SampledPosition, type=NiagaraPosition}}
     Vector Is In Absolute Large World Space:NiagaraBool=true
    )
   ) {isSetParameters=true, module=SetVariables_#1}
   &DynamicMaterialParameters(Module.Param0WriteEnabled=true, Module.Param1WriteEnabled=true, Module.Param2WriteEnabled=true, Module.Param3WriteEnabled=true)
   /SetVariables(
    Particles.ColorSpritesAlt:LinearColor=&MultiplyLinearColor(
     Linear Color=$Particles.ColorSprites
     Scale Factor=&MakeFloat4FromVectorAndFloat("Vector (RGB)"=<1 1 1>, "Float (Alpha)"=0.3)
    )
    Particles.SpriteSizeAlt:Vector2f=&Multiply_Vector2D_ByFloat(Float=1, Vector2D=$Particles.SpriteSize)
   ) {isSetParameters=true}
  ]
  rendererClasses [&NiagaraSpriteRendererProperties, &NiagaraLightRendererProperties]
  renderers [
   {rendererClass=&NiagaraSpriteRendererProperties}
   {rendererClass=&NiagaraLightRendererProperties, rendererIndex=1}
   {rendererClass=&NiagaraSpriteRendererProperties, rendererIndex=2}
  ]
 }
}
summary {
 systemName=NS_Pack_The_Origin_Idle
 emitters [
  {
   bEnabled=false
   emitterName=Cubes_CrashCenter
   rendererClasses [&NiagaraMeshRendererProperties, &NiagaraSpriteRendererProperties]
  }
  {emitterName=SpriteHandGlow, rendererClasses [&NiagaraSpriteRendererProperties]}
  {emitterName=Sprites_Low, rendererClasses [&NiagaraSpriteRendererProperties, &NiagaraLightRendererProperties]}
  {
   emitterName=Cubes_FlowingToCenter
   rendererClasses [&NiagaraMeshRendererProperties, &NiagaraSpriteRendererProperties]
  }
  {emitterName=LightningSprites_Burst001, rendererClasses [&NiagaraSpriteRendererProperties]}
  {emitterName=CenterLightFlicker, rendererClasses [&NiagaraLightRendererProperties]}
  {
   bEnabled=false
   emitterName=Pulse_Cubes_Grid
   rendererClasses [&NiagaraMeshRendererProperties, &NiagaraSpriteRendererProperties]
  }
  {emitterName=Ripple, rendererClasses [&NiagaraMeshRendererProperties]}
  {emitterName=MeshCube01, rendererClasses [&NiagaraMeshRendererProperties]}
  {
   emitterName=CubesShootUpBurst
   rendererClasses [&NiagaraMeshRendererProperties, &NiagaraSpriteRendererProperties]
  }
  {bEnabled=false, emitterName=Decal_ResetBurst, rendererClasses [&NiagaraSpriteRendererProperties]}
  {emitterName=CubeCrashDecals, rendererClasses [&NiagaraSpriteRendererProperties]}
 ]
 userVariables [
  User.bIsFrontEnd {type=NiagaraBool}
  User.bIsFrontEndPreview {type=NiagaraBool}
  User.ColorA {type=LinearColor, description="", defaultValue {struct=LinearColor, value=<rgba 0.5701 0 1 1>}}
  User.ColorB {type=LinearColor, description="", defaultValue {struct=LinearColor, value=<rgba 1 0.9207 0.429 1>}}
  User.SkeletalMesh {
   type=NiagaraDataInterfaceSkeletalMesh
   defaultValue {
    struct=NiagaraExt_VariableValue_DataInterface
    value {dataInterface=&self:NiagaraDataInterfaceSkeletalMesh_0}
   }
  }
 ]
}
