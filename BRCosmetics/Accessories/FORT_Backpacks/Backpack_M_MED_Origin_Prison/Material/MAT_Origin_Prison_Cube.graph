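# Surface material for the Origin Prison backpack cube, drawing a glowing inner ring and core over a fresnel-tinted base.
# Sphere masks on world-aligned screen coordinates shape both glows, distorted by cloud noise sampled through animated vector noise of pre-skinned position.
# Ring, core and base tint sum into emissive, while the core mask lowers specular.
# A static switch adds TheCUBE effect using the ring as alpha, before accessory tech-art overrides.
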Material /BRCosmetics/Accessories/FORT_Backpacks/Backpack_M_MED_Origin_Prison/Material/MAT_Origin_Prison_Cube
graph {
 properties {
  bUseMaterialAttributes=true
  bUsedWithMeshDeformer=true
  bUsedWithNiagaraMeshParticles=true
  bUsedWithSkeletalMesh=true
  pixelDepthOffsetMode=PDOM_Legacy
  usedShadingModels=Default Lit
 }
 nodes [
  n1 = Multiply(A=MaterialFunctionCall!(Center Point World Position) {materialFunction=/Engine/Functions/Engine_MaterialFunctions02/UVs/WorldSpaceAlignedScreenCoordinates}.Offset and Scaled UVS) {constB=80}
  n2 = PreSkinnedPosition()
  n3 = VertexInterpolator(VS=@n2)
  n4 = Time()
  n5 = VectorParameter() {parameterName=NoisePanSpeedAndDirection, defaultValue=<rgba 0.15 0.25 -0.1 1>}
  n6 = Add(
   A=Divide(A=ShaderStageSwitch(PixelShader=@n3.PS, VertexShader=@n2), B=ScalarParameter() {defaultValue=250, parameterName=NoiseDivision})
   B=Multiply(A=Multiply(A=@n4) {constB=0.5}, B=Multiply(A=@n5.RGB) {constB=0.5})
  )
  n7 = TextureSample(UVs=ComponentMask(None=VectorNoise(World Position=@n6) {noiseFunction=VNF_VectorALU}) {g=true, r=true}) {texture=&AtmosphericCloudNoise01}
  n8 = Add(A=@n1, B=Multiply(A=@n7.G) {constA=0.01, constB=0.02})
  n9 = Constant2Vector()
  n10 = Constant() {r=0.25}
  n11 = Constant()
  n12 = Constant2Vector()
  n13 = Constant() {r=0.13}
  n14 = Constant()
  n15 = Power(Base=Subtract(A=SphereMask(A=@n8, B=@n9, Radius=@n10, Hardness=@n11), B=SphereMask(A=@n8, B=@n12, Radius=@n13, Hardness=@n14)))
  n16 = Multiply(A=@n15, B=VectorParameter() {parameterName=Inner Ring Glow Color, defaultValue=<rgba 0.8698 0.0861 0.4824 0>}.RGB)
  n17 = Multiply(B=@n7.G) {constA=0.1}
  n18 = Power(Base=SphereMask(A=Add(A=@n1, B=@n17), B=Constant2Vector(), Radius=Constant() {r=0.12}, Hardness=Constant()))
  n19 = Multiply(A=Multiply(A=@n18) {constB=0.6}, B=VectorParameter() {parameterName=Inner Glow Color, defaultValue=<rgba 0.5175 0.1936 0.8646 0>}.RGB)
  n20 = Constant() {r=3}
  n21 = LinearInterpolate(
   A=VectorParameter() {parameterName=Base Color, defaultValue=<rgba 0.0911 0.0061 0.1667 1>}.RGB
   B=VectorParameter() {parameterName=Fresnel Edge Color, defaultValue=<rgba 0.276 0.0684 0.1949 1>}.RGB
   Alpha=Power(Base=Fresnel(ExponentIn=@n20))
  )
  n22 = SetMaterialAttributes!(MaterialAttributes, Emissive Color=Add(A=Add(A=@n16, B=@n19), B=@n21), Base Color=@n21, Specular=OneMinus(None=@n18)) {attributeSetTypes [#1, #2, #3]}
  MaterialFunctionCall(In=StaticSwitchParameter(True=MaterialFunctionCall!(In=@n22, Alpha=@n15, AO) {materialFunction=/Game/Weapons/WeaponSkins/MaterialFunction/TheCUBE}.Result, False=@n22) {parameterName=Use Cube Effect}) {materialFunction=&MF_TechArtOverrides_accessories}
 ]
}
